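# Physics asset for the StunningMountain male medium armor body, providing secondary motion for its hanging shirt panels and two canisters.
# Capsule bodies form one- or two-segment dyn_ chains, each constrained to a small MSTR anchor capsule with per-joint swing and twist limits, many locking twist.
# Pelvis and thigh capsules plus a large offset root sphere add body shapes alongside the chains, which have no constraints to them.
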PhysicsAsset /BRCosmetics/Characters/Player/Male/Medium/Bodies/M_MED_StunningMountain_Armor/Meshes/M_MED_StunningMountain_Armor_Physics
bodies {
 dyn_canister_bk [{shapeType=Capsule, length=9.0543, radius=4.0155, shapeName=dyn_canister_bk_capsule, center=<4.9277 0 -1.914>, rotation=<rot 90 0 -0.0001>}]
 dyn_canister_bk_MSTR [{shapeType=Capsule, length=1, radius=1, shapeName=dyn_canister_bk_MSTR_capsule}]
 dyn_canister_fr [{
   shapeType=Capsule
   length=9.7978
   radius=3.6691
   shapeName=dyn_canister_fr_capsule
   center=<6.7316 0 -1.9789>
   rotation=<rot 90 179.9999 179.9998>
  }]
 dyn_canister_fr_MSTR [{shapeType=Capsule, length=1, radius=1, shapeName=dyn_canister_fr_MSTR_capsule}]
 dyn_shirt_bk_1_01 [{shapeType=Capsule, length=9.4831, radius=1.7874, shapeName=dyn_shirt_bk_1_01_capsule, center=<9.2748 0.6023 0>, rotation=<rot 90 0 0>}]
 dyn_shirt_bk_1_02 [{shapeType=Capsule, length=5.2705, radius=2.0554, shapeName=dyn_shirt_bk_1_02_capsule, center=<4.9433 0 0>}]
 dyn_shirt_bk_1_MSTR [{shapeType=Capsule, length=1, radius=1, shapeName=dyn_shirt_bk_1_MSTR_capsule}]
 dyn_shirt_bk_2_01 [{shapeType=Capsule, length=10.7982, radius=1.742, shapeName=dyn_shirt_bk_2_01_capsule, center=<7.5169 0 0>, rotation=<rot 90 0 0>}]
 dyn_shirt_bk_2_02 [{shapeType=Capsule, length=4.2375, radius=1.9925, shapeName=dyn_shirt_bk_2_02_capsule, center=<6.1057 0.1626 0>}]
 dyn_shirt_bk_2_MSTR [{shapeType=Capsule, length=1, radius=1, shapeName=dyn_shirt_bk_2_MSTR_capsule}]
 dyn_shirt_bk_3_01 [{shapeType=Capsule, length=8.2535, radius=2.2126, shapeName=dyn_shirt_bk_3_01_capsule, center=<7.5999 0 0>, rotation=<rot 90 0 0>}]
 dyn_shirt_bk_3_02 [{shapeType=Capsule, length=4.5936, radius=2, shapeName=dyn_shirt_bk_3_02_capsule, center=<5.7302 0 0>}]
 dyn_shirt_bk_3_MSTR [{shapeType=Capsule, length=1, radius=1, shapeName=dyn_shirt_bk_3_MSTR_capsule}]
 dyn_shirt_bk_4_01 [{shapeType=Capsule, length=9.8479, radius=3.0888, shapeName=dyn_shirt_bk_4_01_capsule, center=<7.4177 0.7899 0>, rotation=<rot 90 0 0>}]
 dyn_shirt_bk_4_MSTR [{shapeType=Capsule, length=1, radius=1, shapeName=dyn_shirt_bk_4_MSTR_capsule}]
 dyn_shirt_fr_01 [{shapeType=Capsule, length=7.4908, radius=2.8377, shapeName=dyn_shirt_fr_01_capsule, center=<8.128 0.7173 0>, rotation=<rot 90 0 0>}]
 dyn_shirt_fr_02 [{
   shapeType=Capsule
   length=4.3849
   radius=3.1705
   shapeName=dyn_shirt_fr_02_capsule
   center=<4.3998 0 0.0597>
   rotation=<rot 0 90.0007 -178.6089>
  }]
 dyn_shirt_fr_MSTR [{shapeType=Capsule, length=1, radius=1, shapeName=dyn_shirt_fr_MSTR_capsule}]
 dyn_shirt_fr_l_01 [{
   shapeType=Capsule
   length=12.3291
   radius=2.2794
   shapeName=dyn_shirt_fr_l_01_capsule
   center=<9.4014 0 0>
   rotation=<rot 90 -5.0512 -5.0514>
  }]
 dyn_shirt_fr_l_02 [{shapeType=Capsule, length=6.3328, radius=2.2872, shapeName=dyn_shirt_fr_l_02_capsule, center=<4.6129 0 1.1136>}]
 dyn_shirt_fr_l_MSTR [{shapeType=Capsule, length=1, radius=1, shapeName=dyn_shirt_fr_l_MSTR_capsule}]
 dyn_shirt_fr_r_01 [{
   shapeType=Capsule
   length=4.5262
   radius=1.8449
   shapeName=dyn_shirt_fr_r_01_capsule
   center=<9.224 0.8 -0.0294>
   rotation=<rot 0.5535 -179.9999 180>
  }]
 dyn_shirt_fr_r_2_01 [{shapeType=Capsule, length=10.7011, radius=2.6402, shapeName=dyn_shirt_fr_r_2_01_capsule, center=<8.7003 0.7862 0>, rotation=<rot 90 0 0>}]
 dyn_shirt_fr_r_2_MSTR [{shapeType=Capsule, length=1, radius=1, shapeName=dyn_shirt_fr_r_2_MSTR_capsule}]
 dyn_shirt_fr_r_MSTR [{shapeType=Capsule, length=1, radius=1, shapeName=dyn_shirt_fr_r_MSTR_capsule}]
 dyn_shirt_side_01 [{shapeType=Capsule, length=12.5404, radius=2.5254, shapeName=dyn_shirt_side_01_capsule, center=<11.3046 0.7325 0>, rotation=<rot 90 0 0>}]
 dyn_shirt_side_02 [{shapeType=Capsule, length=4.8848, radius=2.6562, shapeName=dyn_shirt_side_02_capsule, center=<4.3448 0 0>, rotation=<rot 0 90 177.9835>}]
 dyn_shirt_side_MSTR [{shapeType=Capsule, length=1, radius=1, shapeName=dyn_shirt_side_MSTR_capsule}]
 dyn_shirt_side_bk_01 [{shapeType=Capsule, length=13.8456, radius=3.0047, shapeName=dyn_shirt_side_bk_01_capsule, center=<9.4839 0.9121 0>, rotation=<rot 90 0 0>}]
 dyn_shirt_side_bk_02 [{
   shapeType=Capsule
   length=6.5765
   radius=2.4789
   shapeName=dyn_shirt_side_bk_02_capsule
   center=<1.9408 0 0>
   rotation=<rot 0.0001 89.9996 89.9996>
  }]
 dyn_shirt_side_bk_MSTR [{shapeType=Capsule, length=1, radius=1, shapeName=dyn_shirt_side_bk_MSTR_capsule}]
 dyn_shirt_side_l_01 [{
   shapeType=Capsule
   length=7.0519
   radius=1.3982
   shapeName=dyn_shirt_side_l_01_capsule
   center=<10.277 0 0>
   rotation=<rot 3.874 -179.9998 -180>
  }]
 dyn_shirt_side_l_02 [{
   shapeType=Capsule
   length=6.644
   radius=1.488
   shapeName=dyn_shirt_side_l_02_capsule
   center=<5.6111 0 -0.057>
   rotation=<rot -10.1764 0.0002 0>
  }]
 dyn_shirt_side_l_MSTR [{shapeType=Capsule, length=1, radius=1, shapeName=dyn_shirt_side_l_MSTR_capsule}]
 pelvis [{shapeType=Capsule, length=7.8985, radius=11.1799, shapeName=pelvis_capsule, center=<-5.8045 0 0>}]
 root [{shapeType=Sphere, radius=62.1807, shapeName=root_sphere, center=<0 -4.837 -58.67>}]
 thigh_l [{shapeType=Capsule, length=26.2804, radius=8.034, shapeName=thigh_l_capsule, center=<-21.6989 0 0>, rotation=<rot 90 180 -179.9999>}]
 thigh_r [{shapeType=Capsule, length=23.709, radius=10.1429, shapeName=thigh_r_capsule, center=<18.5116 -0.1611 -0.4262>, rotation=<rot 90 0 0>}]
}
constraints [
 {bone1Name=dyn_canister_fr, bone2Name=dyn_canister_fr_MSTR, swing1LimitDegrees=20, swing2LimitDegrees=34.32, twistLimitDegrees=5}
 {bone1Name=dyn_canister_bk, bone2Name=dyn_canister_bk_MSTR, swing1LimitDegrees=30.08, swing2LimitDegrees=57.36, twistLimitDegrees=5}
 {
  bone1Name=dyn_shirt_side_bk_01
  bone2Name=dyn_shirt_side_bk_MSTR
  swing1LimitDegrees=60
  swing2LimitDegrees=11.28
  twistLimitDegrees=1
  twistMotion=Locked
 }
 {
  bone1Name=dyn_shirt_side_01
  bone2Name=dyn_shirt_side_MSTR
  swing1LimitDegrees=60
  swing2LimitDegrees=8.4
  twistLimitDegrees=1
  twistMotion=Locked
 }
 {
  bone1Name=dyn_shirt_bk_4_01
  bone2Name=dyn_shirt_bk_4_MSTR
  swing1LimitDegrees=60
  swing2LimitDegrees=5
  twistLimitDegrees=1
  twistMotion=Locked
 }
 {
  bone1Name=dyn_shirt_bk_3_01
  bone2Name=dyn_shirt_bk_3_MSTR
  swing1LimitDegrees=60
  swing2LimitDegrees=10
  twistLimitDegrees=1
  twistMotion=Locked
 }
 {
  bone1Name=dyn_shirt_bk_2_01
  bone2Name=dyn_shirt_bk_2_MSTR
  swing1LimitDegrees=60
  swing2LimitDegrees=10
  twistLimitDegrees=1
  twistMotion=Locked
 }
 {
  bone1Name=dyn_shirt_bk_1_01
  bone2Name=dyn_shirt_bk_1_MSTR
  swing1LimitDegrees=60
  swing2LimitDegrees=10
  twistLimitDegrees=1
  twistMotion=Locked
 }
 {bone1Name=dyn_shirt_fr_r_01, bone2Name=dyn_shirt_fr_r_MSTR, swing1LimitDegrees=60, swing2LimitDegrees=30, twistLimitDegrees=5}
 {bone1Name=dyn_shirt_fr_r_2_01, bone2Name=dyn_shirt_fr_r_2_MSTR, swing1LimitDegrees=70, swing2LimitDegrees=30, twistLimitDegrees=1}
 {bone1Name=dyn_shirt_fr_01, bone2Name=dyn_shirt_fr_MSTR, swing1LimitDegrees=90, swing2LimitDegrees=20, twistLimitDegrees=5}
 {bone1Name=dyn_shirt_fr_02, bone2Name=dyn_shirt_fr_01, swing1LimitDegrees=30, swing2LimitDegrees=10, twistLimitDegrees=10}
 {
  bone1Name=dyn_shirt_side_bk_02
  bone2Name=dyn_shirt_side_bk_01
  swing1LimitDegrees=30
  swing2LimitDegrees=10
  twistLimitDegrees=45
  twistMotion=Locked
 }
 {
  bone1Name=dyn_shirt_side_02
  bone2Name=dyn_shirt_side_01
  swing1LimitDegrees=27.12
  swing2LimitDegrees=10
  twistLimitDegrees=45
  twistMotion=Locked
 }
 {
  bone1Name=dyn_shirt_side_l_02
  bone2Name=dyn_shirt_side_l_01
  swing1LimitDegrees=30
  swing2LimitDegrees=5
  twistLimitDegrees=45
  twistMotion=Locked
 }
 {bone1Name=dyn_shirt_side_l_01, bone2Name=dyn_shirt_side_l_MSTR, swing1LimitDegrees=60, swing2LimitDegrees=10, twistLimitDegrees=30}
 {bone1Name=dyn_shirt_fr_l_02, bone2Name=dyn_shirt_fr_l_01, swing1LimitDegrees=5, swing2LimitDegrees=5, twistLimitDegrees=1}
 {
  bone1Name=dyn_shirt_fr_l_01
  bone2Name=dyn_shirt_fr_l_MSTR
  swing1LimitDegrees=16.8
  swing2LimitDegrees=10
  twistLimitDegrees=1
  twistMotion=Locked
 }
 {bone1Name=dyn_shirt_bk_3_02, bone2Name=dyn_shirt_bk_3_01, swing1LimitDegrees=20, swing2LimitDegrees=10, twistLimitDegrees=1}
 {bone1Name=dyn_shirt_bk_2_02, bone2Name=dyn_shirt_bk_2_01, swing1LimitDegrees=10, swing2LimitDegrees=10, twistLimitDegrees=1}
 {bone1Name=dyn_shirt_bk_1_02, bone2Name=dyn_shirt_bk_1_01, swing1LimitDegrees=10, swing2LimitDegrees=10, twistLimitDegrees=1}
]
