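# Physics asset for the Frontend variant of the StunningMountain medium male armor, setting up secondary motion for two canisters and the shirt panels around the front, back and sides.
# Each one- or two-segment capsule chain is constrained to a tiny MSTR anchor body.
# Swing limits are mostly hinge-like, wider on one axis, with tight or locked twist on most joints.
# Unconstrained pelvis and thigh capsules provide body collision shapes.
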PhysicsAsset /BRCosmetics/Characters/Player/Male/Medium/Bodies/M_MED_StunningMountain_Armor/Meshes/M_MED_StunningMountain_Armor_Physics_Frontend
bodies {
 dyn_canister_bk [{shapeType=Capsule, length=9.0543, radius=4.0155, shapeName=dyn_canister_bk_capsule, center=<4.9277 0 -1.914>, rotation=<rot 90 0 -0.0001>}]
 dyn_canister_bk_MSTR [{shapeType=Capsule, length=1, radius=1, shapeName=dyn_canister_bk_MSTR_capsule}]
 dyn_canister_fr [{
   shapeType=Capsule
   length=9.7978
   radius=3.6691
   shapeName=dyn_canister_fr_capsule
   center=<6.7316 0 -1.9789>
   rotation=<rot 90 179.9999 179.9998>
  }]
 dyn_canister_fr_MSTR [{shapeType=Capsule, length=1, radius=1, shapeName=dyn_canister_fr_MSTR_capsule}]
 dyn_shirt_bk_1_01 [{shapeType=Capsule, length=8.9831, radius=2.2874, shapeName=dyn_shirt_bk_1_01_capsule, center=<9.2748 0.6023 0>, rotation=<rot 90 0 0>}]
 dyn_shirt_bk_1_02 [{shapeType=Capsule, length=6.7687, radius=5.0554, shapeName=dyn_shirt_bk_1_02_capsule, center=<5 2.5 0>}]
 dyn_shirt_bk_1_MSTR [{shapeType=Capsule, length=1, radius=1, shapeName=dyn_shirt_bk_1_MSTR_capsule}]
 dyn_shirt_bk_2_01 [{shapeType=Capsule, length=9.8684, radius=2.6719, shapeName=dyn_shirt_bk_2_01_capsule, center=<9.758 0 0>, rotation=<rot 90 0 0>}]
 dyn_shirt_bk_2_02 [{shapeType=Capsule, length=5.3109, radius=2.3973, shapeName=dyn_shirt_bk_2_02_capsule, center=<6.1057 0 0>}]
 dyn_shirt_bk_2_MSTR [{shapeType=Capsule, length=1, radius=1, shapeName=dyn_shirt_bk_2_MSTR_capsule}]
 dyn_shirt_bk_3_01 [{shapeType=Capsule, length=8.2535, radius=2.2126, shapeName=dyn_shirt_bk_3_01_capsule, center=<7.5999 0 0>, rotation=<rot 90 0 0>}]
 dyn_shirt_bk_3_02 [{shapeType=Capsule, length=4.5936, radius=2, shapeName=dyn_shirt_bk_3_02_capsule, center=<5.7302 0 0>}]
 dyn_shirt_bk_3_MSTR [{shapeType=Capsule, length=1, radius=1, shapeName=dyn_shirt_bk_3_MSTR_capsule}]
 dyn_shirt_bk_4_01 [{shapeType=Capsule, length=9.8479, radius=3.0888, shapeName=dyn_shirt_bk_4_01_capsule, center=<7.4177 0.7899 0>, rotation=<rot 90 0 0>}]
 dyn_shirt_bk_4_MSTR [{shapeType=Capsule, length=1, radius=1, shapeName=dyn_shirt_bk_4_MSTR_capsule}]
 dyn_shirt_fr_01 [{
   shapeType=Capsule
   length=8.8504
   radius=2.6916
   shapeName=dyn_shirt_fr_01_capsule
   center=<11.8718 0.8645 0>
   rotation=<rot -90 -5.051 5.0513>
  }]
 dyn_shirt_fr_MSTR [{shapeType=Capsule, length=1, radius=1, shapeName=dyn_shirt_fr_MSTR_capsule}]
 dyn_shirt_fr_l_01 [{
   shapeType=Capsule
   length=9.1758
   radius=2.5284
   shapeName=dyn_shirt_fr_l_01_capsule
   center=<9.753 0.2372 0>
   rotation=<rot -90 179.8418 -179.8415>
  }]
 dyn_shirt_fr_l_02 [{shapeType=Capsule, length=4.9407, radius=2.6001, shapeName=dyn_shirt_fr_l_02_capsule, center=<4.7277 0.2372 0>}]
 dyn_shirt_fr_l_MSTR [{shapeType=Capsule, length=1, radius=1, shapeName=dyn_shirt_fr_l_MSTR_capsule}]
 dyn_shirt_fr_r_01 [{shapeType=Capsule, length=3.0981, radius=2.2333, shapeName=dyn_shirt_fr_r_01_capsule, center=<6.6161 0 0>, rotation=<rot 0 0.0004 0>}]
 dyn_shirt_fr_r_2_01 [{shapeType=Capsule, length=7.687, radius=3.5078, shapeName=dyn_shirt_fr_r_2_01_capsule, center=<10.9796 0 0>, rotation=<rot 90 3.0357 3.0356>}]
 dyn_shirt_fr_r_2_MSTR [{shapeType=Capsule, length=1, radius=1, shapeName=dyn_shirt_fr_r_2_MSTR_capsule}]
 dyn_shirt_fr_r_MSTR [{shapeType=Capsule, length=1, radius=1, shapeName=dyn_shirt_fr_r_MSTR_capsule}]
 dyn_shirt_side_01 [{shapeType=Capsule, length=6.0404, radius=7.0254, shapeName=dyn_shirt_side_01_capsule, center=<11 6 0>, rotation=<rot 90 0 0>}]
 dyn_shirt_side_02 [{shapeType=Capsule, length=3.2697, radius=3.1562, shapeName=dyn_shirt_side_02_capsule, center=<4.3448 2 0>, rotation=<rot 0 90 177.9835>}]
 dyn_shirt_side_MSTR [{shapeType=Capsule, length=1, radius=1, shapeName=dyn_shirt_side_MSTR_capsule}]
 dyn_shirt_side_bk_01 [{shapeType=Capsule, length=11.3043, radius=5.5459, shapeName=dyn_shirt_side_bk_01_capsule, center=<9.5 4 0>, rotation=<rot 90 0 0>}]
 dyn_shirt_side_bk_02 [{
   shapeType=Capsule
   length=4.3223
   radius=3.4789
   shapeName=dyn_shirt_side_bk_02_capsule
   center=<3.5 1.8 0>
   rotation=<rot 0.0001 89.9996 179.9996>
  }]
 dyn_shirt_side_bk_MSTR [{shapeType=Capsule, length=1, radius=1, shapeName=dyn_shirt_side_bk_MSTR_capsule}]
 dyn_shirt_side_l_01 [{
   shapeType=Capsule
   length=5.0942
   radius=2
   shapeName=dyn_shirt_side_l_01_capsule
   center=<10.6314 0 -0.0754>
   rotation=<rot 89.0967 -0.0037 -0.0039>
  }]
 dyn_shirt_side_l_02 [{shapeType=Capsule, length=3.845, radius=2.6297, shapeName=dyn_shirt_side_l_02_capsule, center=<3.8685 0 0>}]
 dyn_shirt_side_l_MSTR [{shapeType=Capsule, length=1, radius=1, shapeName=dyn_shirt_side_l_MSTR_capsule}]
 pelvis [{shapeType=Capsule, length=7.8985, radius=11.1799, shapeName=pelvis_capsule, center=<-5.8045 0 0>}]
 thigh_l [{shapeType=Capsule, length=26.821, radius=7.4934, shapeName=thigh_l_capsule, center=<-21.6989 0 0>, rotation=<rot 90 180 -179.9999>}]
 thigh_r [{shapeType=Capsule, length=24.2496, radius=9.6023, shapeName=thigh_r_capsule, center=<18.5116 -0.1611 -0.4262>, rotation=<rot 90 0 0>}]
}
constraints [
 {bone1Name=dyn_canister_fr, bone2Name=dyn_canister_fr_MSTR, swing1LimitDegrees=9.92, swing2LimitDegrees=27.12, twistLimitDegrees=5}
 {bone1Name=dyn_canister_bk, bone2Name=dyn_canister_bk_MSTR, swing1LimitDegrees=30.08, swing2LimitDegrees=57.36, twistLimitDegrees=5}
 {
  bone1Name=dyn_shirt_side_bk_01
  bone2Name=dyn_shirt_side_bk_MSTR
  swing1LimitDegrees=70
  swing2LimitDegrees=2
  twistLimitDegrees=1
  twistMotion=Locked
 }
 {
  bone1Name=dyn_shirt_side_01
  bone2Name=dyn_shirt_side_MSTR
  swing1LimitDegrees=50
  swing2LimitDegrees=2
  twistLimitDegrees=1
  twistMotion=Locked
 }
 {
  bone1Name=dyn_shirt_bk_4_01
  bone2Name=dyn_shirt_bk_4_MSTR
  swing1LimitDegrees=60
  swing2LimitDegrees=5
  twistLimitDegrees=1
  twistMotion=Locked
 }
 {
  bone1Name=dyn_shirt_bk_3_01
  bone2Name=dyn_shirt_bk_3_MSTR
  swing1LimitDegrees=60
  swing2LimitDegrees=10
  twistLimitDegrees=1
  twistMotion=Locked
 }
 {bone1Name=dyn_shirt_bk_2_01, bone2Name=dyn_shirt_bk_2_MSTR, swing1LimitDegrees=80, swing2LimitDegrees=2, twistLimitDegrees=1}
 {bone1Name=dyn_shirt_bk_1_01, bone2Name=dyn_shirt_bk_1_MSTR, swing1LimitDegrees=70, swing2LimitDegrees=2, twistLimitDegrees=0}
 {
  bone1Name=dyn_shirt_side_bk_02
  bone2Name=dyn_shirt_side_bk_01
  swing1LimitDegrees=10
  swing2LimitDegrees=2
  twistLimitDegrees=45
  twistMotion=Locked
 }
 {bone1Name=dyn_shirt_side_02, bone2Name=dyn_shirt_side_01, swing1LimitDegrees=10, swing2LimitDegrees=2, twistLimitDegrees=0}
 {bone1Name=dyn_shirt_bk_3_02, bone2Name=dyn_shirt_bk_3_01, swing1LimitDegrees=20, swing2LimitDegrees=10, twistLimitDegrees=1}
 {bone1Name=dyn_shirt_bk_2_02, bone2Name=dyn_shirt_bk_2_01, swing1LimitDegrees=10, swing2LimitDegrees=5, twistLimitDegrees=1}
 {bone1Name=dyn_shirt_bk_1_02, bone2Name=dyn_shirt_bk_1_01, swing1LimitDegrees=10, swing2LimitDegrees=2, twistLimitDegrees=10}
 {bone1Name=dyn_shirt_fr_r_01, bone2Name=dyn_shirt_fr_r_MSTR, swing1LimitDegrees=10, swing2LimitDegrees=10, twistLimitDegrees=1}
 {bone1Name=dyn_shirt_fr_r_2_01, bone2Name=dyn_shirt_fr_r_2_MSTR, swing1LimitDegrees=40, swing2LimitDegrees=2, twistLimitDegrees=1}
 {bone1Name=dyn_shirt_fr_01, bone2Name=dyn_shirt_fr_MSTR, swing1LimitDegrees=30, swing2LimitDegrees=10, twistLimitDegrees=1}
 {bone1Name=dyn_shirt_fr_l_02, bone2Name=dyn_shirt_fr_l_01, swing1LimitDegrees=30, swing2LimitDegrees=10, twistLimitDegrees=1}
 {bone1Name=dyn_shirt_fr_l_01, bone2Name=dyn_shirt_fr_l_MSTR, swing1LimitDegrees=30, swing2LimitDegrees=10, twistLimitDegrees=1}
 {bone1Name=dyn_shirt_side_l_02, bone2Name=dyn_shirt_side_l_01, swing1LimitDegrees=30, swing2LimitDegrees=10, twistLimitDegrees=1}
 {bone1Name=dyn_shirt_side_l_01, bone2Name=dyn_shirt_side_l_MSTR, swing1LimitDegrees=30, swing2LimitDegrees=10, twistLimitDegrees=1}
]
