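# Animation Blueprint for the StunningMountain male medium armor cosmetic, adding secondary motion to the pose copied from its parent mesh.
# A crouch-aware Control Rig pose feeds RigidBody simulations in spine_01 space, with a separate front-end physics asset selected by bIsInFrontEnd.
# The gameplay branch adds corrective shirt and harness bone offsets after physics.
# The event graph picks gravity by backbling state and derives skirt alphas from movement direction and thigh rotation.
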AnimBlueprint /BRCosmetics/Characters/Player/Male/Medium/Bodies/M_MED_StunningMountain_Armor/Meshes/M_MED_StunningMountain_Armor_AnimBP
target_skeleton=&/M_MED_StunningMountain_Armor_Skeleton
graphs [
 AnimGraph {nodes [
   n1 = RigidBody!(ComponentPose=Local To Component(LocalPose=Use cached pose 'CtrlRig_Pose'()), Alpha, OverrideWorldGravity:Vector="0,0,-980", ExternalForce) -> (Pose) {
    bClampLinearTranslationLimitToRefPose=true
    bOverrideWorldGravity=true
    overridePhysicsAsset=&/M_MED_StunningMountain_Armor_Physics_Frontend
    simulationSpace=BaseBoneSpace
    alphaBoolBlend {blendOption=Linear}
    baseBoneRef {boneName=spine_01}
    componentAppliedLinearAccClamp=<2000 2000 2000>
    componentLinearAccScale=<0.3 0.3 0.3>
    componentLinearVelScale=<0.3 0.3 0.3>
   }
   n2 = RigidBody(
    ComponentPose=Local To Component(LocalPose=Use cached pose 'CtrlRig_Pose'())
    OverrideWorldGravity=Get Override World Gravity() -> (Override World Gravity:Vector)
   ) {
    bClampLinearTranslationLimitToRefPose=true
    bOverrideWorldGravity=true
    overridePhysicsAsset=&/M_MED_StunningMountain_Armor_Physics
    simulationSpace=BaseBoneSpace
    alphaBoolBlend {blendOption=Linear}
    baseBoneRef {boneName=spine_01}
    componentAppliedLinearAccClamp=<2000 2000 2000>
    componentLinearAccScale=<0.3 0.3 0.3>
    componentLinearVelScale=<0.3 0.3 0.3>
   }
   n3 = "Transform (Modify) Bone - Bone: dyn_shirt_fr_l_MSTR"(ComponentPose=@n2, Alpha=1, Translation="0,0,0", Rotation="0,5,0", Scale="1,1,1") {
    rotationMode=BMM_Additive
    rotationSpace=BCS_BoneSpace
    translationMode=BMM_Additive
    translationSpace=BCS_BoneSpace
    alphaBoolBlend {blendOption=Linear}
    alphaScaleBiasClamp {bInterpResult=true, interpSpeedDecreasing=5, interpSpeedIncreasing=5}
    boneToModify {boneName=dyn_shirt_fr_l_MSTR}
    rotation {yaw=5}
   }
   n4 = "Transform (Modify) Bone - Bone: dyn_shirt_fr_l_MSTR"(ComponentPose=@n3, Alpha=1, Translation="0,3,0", Rotation="0,0,0", Scale="1,1,1") {
    rotationMode=BMM_Additive
    rotationSpace=BCS_BoneSpace
    translationMode=BMM_Additive
    translationSpace=BCS_BoneSpace
    alphaBoolBlend {blendOption=Linear}
    alphaScaleBiasClamp {bInterpResult=true, interpSpeedDecreasing=5, interpSpeedIncreasing=5}
    boneToModify {boneName=dyn_shirt_fr_l_MSTR}
    translation {y=3}
   }
   n5 = "Transform (Modify) Bone - Bone: dyn_shirt_side_l_MSTR"(ComponentPose=@n4, Alpha=1, Translation="0,0,0", Rotation="-15,-15,-15", Scale="1,1,1") {
    rotationMode=BMM_Additive
    rotationSpace=BCS_BoneSpace
    translationMode=BMM_Additive
    translationSpace=BCS_BoneSpace
    alphaBoolBlend {blendOption=CubicInOut}
    alphaScaleBiasClamp {bInterpResult=true, interpSpeedDecreasing=8, interpSpeedIncreasing=8}
    boneToModify {boneName=dyn_shirt_side_l_MSTR}
    rotation=<rot -15 -15 -15>
   }
   n6 = "Transform (Modify) Bone - Bone: harness_l"!(ComponentPose=@n5, bAlphaBoolEnabled:bool=True, Translation="9.141,1.546,-2", Rotation="26.350,0.031,-0.003", Scale="1,1,1") -> (Pose) {
    alphaInputType=Bool
    rotationMode=BMM_Additive
    rotationSpace=BCS_BoneSpace
    translationMode=BMM_Additive
    translationSpace=BCS_BoneSpace
    alphaBoolBlend {blendOption=Linear}
    alphaScaleBiasClamp {bInterpResult=true, interpSpeedDecreasing=5, interpSpeedIncreasing=5}
    boneToModify {boneName=harness_l}
    rotation=<rot 26.35 0.031 -0.003>
    translation=<9.141 1.546 -2>
   }
   Output Pose(Result=Blend Poses by bool(BlendPose_0=Component To Local(ComponentPose=@n1), BlendPose_1=Component To Local(ComponentPose=@n6), bActiveValue=Get bIsInFrontEnd()) {blendType=Linear}) {layerGroup=None}
   Save cached pose 'BasePose'(Pose=Copy Pose From Mesh() {bUseAttachedParent=true}) {cacheName=BasePose}
   n7 = Control Rig!(Source=Use cached pose 'BasePose'(), Alpha, isCrouching:bool=Get bIsCrouching(), isCrouchingSprint:bool=Get bIsCrouchSprinting()) -> (Pose) {
    alphaBoolBlend {blendOption=Linear}
    controlRigAssetReference {blueprintRigClass=&/M_MED_StunningMountain_Armor_CtrlRig.M_MED_StunningMountain_Armor_CtrlRig_C}
   }
   Save cached pose 'CtrlRig_Pose'(Pose=@n7) {cacheName=CtrlRig_Pose}
  ]}
 EventGraph {nodes [
   n1 = Event OnCharacterCustomizationLoaded()
   > Set BodyMesh(BodyMesh=@n1.CustomBaseMesh)
   Event BlueprintUpdateAnimation()
   > n2 = Sequence()
   > n3 = Gravity Override State Function DEPRECATED(
    CharacterPartAnimInstance=Self-Reference()
    GameTimeInSeconds=Get Game Time in Seconds() -> (Game Time in Seconds:double)
    GravityOverrideParams=Get Gravity Override Params()
   )
   > Set Override World Gravity(Override World Gravity:Vector=SelectVector(A="0,0,-980", B=@n3, bPickA=CheckForBackbling().bIsBackblingLong)) -> (Output_Get:Vector)
   n4 = OR Boolean(A=Get bIsCrouching(), B=Get bIsCrouchSprinting())
   n5 = Dot Product!(
    A:Vector=GetActorForwardVector!(self:Actor Object Reference=TryGetPawnOwner()) -> (ReturnValue:Vector)
    B:Vector=Normalize!(A:Vector=Get PawnVelocity(), Tolerance:float=0.0001) -> (ReturnValue:Vector) {is=title K2Node_CallFunction}
   ) -> (ReturnValue:double)
   Set Skirt_on_Moving(
    execute=@n2.then_1
    Skirt_on_Moving:double=SelectFloat(A=MapRangeClamped(Value=@n5, InRangeA=0, InRangeB=-0.9), B=0, bPickA=NOT Boolean(A=@n4))
   ) -> (Output_Get:double)
   > Set Skirt_on_Crouch_Side(Skirt_on_Crouch_Side:double=GetAlphaFromReferenceTransformDelta(
     SkeletalMeshComponent=Get BodyMesh() -> (BodyMesh:Skeletal Mesh Component Object Reference)
     BoneName=thigh_l
     BaseName=pelvis
     DeltaComponentType=EulerX
     InputRange=(X=-13,Y=50)
     TwistAxis="1,0,0"
    )) -> (Output_Get:double)
   n6 = Get BodyMesh()
   n7 = NOT Boolean(A=Get bIsMoving2D())
   n8 = Get bIsCrouching()
   > Set Skirt_on_Crouching(Skirt_on_Crouching:double=SelectFloat(
     A=GetAlphaFromReferenceTransformDelta(SkeletalMeshComponent=@n6, BoneName=thigh_l, BaseName=pelvis, DeltaComponentType=EulerY, InputRange=(X=0,Y=30), TwistAxis="1,0,0")
     B=0
     bPickA=AND Boolean(A=@n7, B=@n8)
    )) -> (Output_Get:double)
  ]}
]
variables [
 BodyMesh {type=SkeletalMeshComponent}
 Game Time in Seconds {type=number}
 Gravity Override Params {type=GravityOverrideParamsStruct}
 Override World Gravity {type=Vector}
 Skirt_on_Crouch_Side {type=number}
 Skirt_on_Crouching {type=number}
 Skirt_on_Moving {type=number}
]
