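# Control Rig for the StunningMountain male armor's face accessory, applying corrective offsets to hair bones as pose and state change.
# Spherical pose readers measure head orientation against neck_01 and spine_05, while crouch, movement and targeting flags become smoothed alphas.
# These values, alone or multiplied, weight chained ModifyTransforms on back, side and strand hair.
# Debug prints are disabled, two readers still draw debug, and several offsets are identity.
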ControlRigBlueprint /BRCosmetics/Characters/Player/Male/Medium/Bodies/M_MED_StunningMountain_Armor/Meshes/Parts/M_MED_StunningMountain_Armor_FaceAcc_CtrlRig
graphs [
 RigVMFunctionLibrary {links [], nodes []}
 RigVMModel {nodes [
   RigUnit_BeginExecution = &RigUnit_BeginExecution()
   n1 = &RigUnit_SphericalPoseReader(
    ActiveRegionScaleFactors=(PositiveWidth=1,NegativeWidth=1,PositiveHeight=1,NegativeHeight=1)
    ActiveRegionSize=0.05
    Debug=(bDrawDebug=false,bDraw2D=false,bDrawLocalAxes=false,DebugScale=25,DebugSegments=20,DebugThickness=0.25)
    DriverAxis=(X=0,Y=-1,Z=0)
    DriverItem=(Type=Bone,Name="head")
    FalloffRegionScaleFactors=(PositiveWidth=0.1,NegativeWidth=0.2,PositiveHeight=0.1,NegativeHeight=0.2)
    FalloffSize=0.45
    FlipHeightScaling=false
    FlipWidthScaling=false
    OptionalParentItem=(Type=Bone,Name="neck_01")
    RotationOffset=(X=40.727,Y=70.1225,Z=100.6829)
    ExecutePin=@RigUnit_BeginExecution.ExecutePin
   )
   RigVMFunction_MathBoolNot = &RigVMFunction_MathBoolNot(Value=RigVMVariableNode() {variable=IsCrouching}.Value)
   n2 = &RigVMFunction_AlphaInterp(
    InterpSpeedDecreasing=10
    InterpSpeedIncreasing=10
    Scale=1
    bClampResult=False
    Value=RigVMDispatchNode(Values=(0,1), Index=&RigVMFunction_MathBoolToInteger(Value=@RigVMFunction_MathBoolNot.Result).Result) {template="DISPATCH_RigVMDispatch_SelectInt32(in Index,in Values,out Result)"}.Result
   )
   > RigVMVariableNode(Value=@n2.Result) {variable=IsNotCrouchingAlpha}
   n3 = &RigVMFunction_MathBoolNot(Value=RigVMVariableNode() {variable=IsMoving2d}.Value)
   n4 = &RigVMFunction_AlphaInterp(
    InterpSpeedDecreasing=10
    InterpSpeedIncreasing=10
    Scale=1
    bClampResult=False
    Value=RigVMDispatchNode(Values=(0,1), Index=&RigVMFunction_MathBoolToInteger(Value=@n3.Result).Result) {template="DISPATCH_RigVMDispatch_SelectInt32(in Index,in Values,out Result)"}.Result
   )
   > RigVMVariableNode(Value=@n4.Result) {variable=IsMoving2dAlpha}
   n5 = &RigVMFunction_MathDoubleMul(B=1, A=@n1.OutputParam)
   > &RigUnit_ModifyTransforms(
    ItemToModify=((Item=(Type=Bone,Name="mstr_dyn_hairBack_00"),Transform=(Rotation=(X=0,Y=-0.0872,Z=0,W=0.9962),Translation=(X=0,Y=0,Z=0.1756),Scale3D=(X=1,Y=1,Z=1))))
    WeightMinimum=0
    Weight=@n5.Result
   )
   > &RigUnit_ModifyTransforms(
    ItemToModify=((Item=(Type=Bone,Name="mstr_dyn_hair_00_r"),Transform=(Rotation=(X=0,Y=0,Z=0,W=1),Translation=(X=0,Y=0,Z=-2),Scale3D=(X=1,Y=1,Z=1))))
    WeightMinimum=0
    Weight=@n5.Result
   )
   > &RigUnit_ModifyTransforms(
    ItemToModify=((Item=(Type=Bone,Name="mstr_dyn_hair_00_l"),Transform=(Rotation=(X=0,Y=0,Z=0,W=1),Translation=(X=0,Y=0,Z=0),Scale3D=(X=1,Y=1,Z=1))))
    WeightMinimum=0
    Weight=@n5.Result
   )
   > &RigUnit_ModifyTransforms(
    ItemToModify=((Item=(Type=Bone,Name="dyn_hairBack_01"),Transform=(Rotation=(X=0,Y=0,Z=0,W=1),Translation=(X=0,Y=0,Z=0),Scale3D=(X=1,Y=1,Z=1))))
    WeightMinimum=0
    Weight=@n5.Result
   )
   > DISPATCH_RigVMDispatch_Print = RigVMDispatchNode(Enabled=false, Prefix=Up_, ScreenColor=(R=1,G=1,B=1,A=1), ScreenDuration=0.05, Value=@n5.Result) {template="DISPATCH_RigVMDispatch_Print(in Prefix,in Value,in Enabled,in ScreenDuration,in ScreenColor)"}
   > &RigUnit_SphericalPoseReader(
    ActiveRegionScaleFactors=(PositiveWidth=1,NegativeWidth=1,PositiveHeight=1,NegativeHeight=1)
    ActiveRegionSize=0.16
    Debug=(bDrawDebug=false,bDraw2D=false,bDrawLocalAxes=false,DebugScale=25,DebugSegments=20,DebugThickness=0.25)
    DriverAxis=(X=0,Y=-1,Z=0)
    DriverItem=(Type=Bone,Name="head")
    FalloffRegionScaleFactors=(PositiveWidth=1,NegativeWidth=1,PositiveHeight=1,NegativeHeight=0.1)
    FalloffSize=0.45
    FlipHeightScaling=false
    FlipWidthScaling=false
    OptionalParentItem=(Type=Bone,Name="spine_05")
    RotationOffset=(X=40.727,Y=70.1225,Z=97.5692)
   )
   n6 = &RigVMFunction_MathBoolNot(Value=RigVMVariableNode() {variable=IsMoving2d}.Value)
   n7 = &RigVMFunction_AlphaInterp(
    InterpSpeedDecreasing=4
    InterpSpeedIncreasing=1
    Scale=1
    bClampResult=False
    Value=RigVMDispatchNode(Values=(1,0), Index=&RigVMFunction_MathBoolToInteger(Value=@n6.Result).Result) {template="DISPATCH_RigVMDispatch_SelectInt32(in Index,in Values,out Result)"}.Result
   )
   n8 = &RigVMFunction_MathDoubleMul(A=1, B=&RigVMFunction_MathDoubleMul(A=@n7.Result, B=RigVMVariableNode() {variable=IsNotCrouchingAlpha}.Value).Result)
   > RigVMDispatchNode(Enabled=false, Prefix=Moving, ScreenColor=(R=1,G=1,B=1,A=1), ScreenDuration=0.05, Value=@n8.Result) {template="DISPATCH_RigVMDispatch_Print(in Prefix,in Value,in Enabled,in ScreenDuration,in ScreenColor)"}
   > &RigUnit_ModifyTransforms(
    ItemToModify=((Item=(Type=Bone,Name="mstr_dyn_hairBack_00"),Transform=(Rotation=(X=0,Y=0,Z=0,W=1),Translation=(X=0,Y=0,Z=0),Scale3D=(X=1,Y=1,Z=1))),(Item=(Type=Bone,Name="None"),Transform=(Rotation=(X=0,Y=0,Z=0,W=1),Translation=(X=0,Y=0,Z=0),Scale3D=(X=1,Y=1,Z=1))))
    WeightMinimum=0
    Weight=@n8.Result
   )
   > &RigUnit_ModifyTransforms(
    ItemToModify=((Item=(Type=Bone,Name="dyn_hairBack_01"),Transform=(Rotation=(X=0,Y=0,Z=0,W=1),Translation=(X=0,Y=0,Z=0),Scale3D=(X=1,Y=1,Z=1))),(Item=(Type=Bone,Name="None"),Transform=(Rotation=(X=0,Y=0,Z=0,W=1),Translation=(X=0,Y=0,Z=0),Scale3D=(X=1,Y=1,Z=1))))
    WeightMinimum=0
    Weight=@n8.Result
   )
   n9 = &RigVMFunction_MathBoolNot(Value=RigVMVariableNode() {variable=isTargeting}.Value)
   n10 = &RigVMFunction_AlphaInterp(
    InterpSpeedDecreasing=10
    InterpSpeedIncreasing=10
    Scale=1
    bClampResult=False
    Value=RigVMDispatchNode(Values=(1,0), Index=&RigVMFunction_MathBoolToInteger(Value=@n9.Result).Result) {template="DISPATCH_RigVMDispatch_SelectInt32(in Index,in Values,out Result)"}.Result
   )
   > RigVMDispatchNode(Enabled=false, Prefix=Target, ScreenColor=(R=1,G=1,B=1,A=1), ScreenDuration=0.05, Value=@n10.Result) {template="DISPATCH_RigVMDispatch_Print(in Prefix,in Value,in Enabled,in ScreenDuration,in ScreenColor)"}
   > RigVMVariableNode(Value=@n10.Result) {variable=isTargetingAlpha}
   > n11 = &RigUnit_SphericalPoseReader(
    ActiveRegionScaleFactors=(PositiveWidth=1,NegativeWidth=2,PositiveHeight=1,NegativeHeight=3)
    ActiveRegionSize=0.05
    Debug=(bDrawDebug=false,bDraw2D=false,bDrawLocalAxes=false,DebugScale=25,DebugSegments=20,DebugThickness=0.25)
    DriverAxis=(X=0,Y=-1,Z=0)
    DriverItem=(Type=Bone,Name="head")
    FalloffRegionScaleFactors=(PositiveWidth=0.1,NegativeWidth=0.3,PositiveHeight=0.1,NegativeHeight=0.3)
    FalloffSize=0.45
    FlipHeightScaling=false
    FlipWidthScaling=false
    OptionalParentItem=(Type=Bone,Name="neck_01")
    RotationOffset=(X=40.727,Y=70.1225,Z=100.6829)
   )
   Multiply = &RigVMFunction_MathDoubleMul(A=@n11.OutputParam, B=RigVMVariableNode() {variable=isTargetingAlpha}.Value)
   n12 = &RigVMFunction_MathDoubleMul(A=@Multiply.Result, B=RigVMVariableNode() {variable=IsNotCrouchingAlpha}.Value)
   > RigVMDispatchNode(Enabled=false, Prefix=Target, ScreenColor=(R=1,G=1,B=1,A=1), ScreenDuration=0.05, Value=@n12.Result) {template="DISPATCH_RigVMDispatch_Print(in Prefix,in Value,in Enabled,in ScreenDuration,in ScreenColor)"}
   > &RigUnit_ModifyTransforms(
    ItemToModify=((Item=(Type=Bone,Name="mstr_dyn_hairBack_00"),Transform=(Rotation=(X=-0.342,Y=-0.1632,Z=0,W=0.9254),Translation=(X=0.3878,Y=0,Z=0.5083),Scale3D=(X=1,Y=1,Z=1))))
    WeightMinimum=0
    Weight=@n12.Result
   )
   > &RigUnit_ModifyTransforms(
    ItemToModify=((Item=(Type=Bone,Name="mstr_dyn_hair_00_r"),Transform=(Rotation=(X=0,Y=0,Z=0.4226,W=0.9063),Translation=(X=-0.0716,Y=-2.9674,Z=0),Scale3D=(X=1,Y=1,Z=1))))
    WeightMinimum=0
    Weight=@n12.Result
   )
   > &RigUnit_ModifyTransforms(
    ItemToModify=((Item=(Type=Bone,Name="mstr_dyn_hairBack_00"),Transform=(Rotation=(X=0,Y=0,Z=0,W=1),Translation=(X=0,Y=0,Z=0),Scale3D=(X=1,Y=1,Z=1))))
    WeightMinimum=0
    Weight=@n12.Result
   )
   n13 = &RigVMFunction_AlphaInterp(ClampMin=0.5, InterpSpeedDecreasing=10, InterpSpeedIncreasing=10, Scale=1, bClampResult=false, Value=@n12.Result)
   > RigVMDispatchNode(Enabled=false, Prefix=BangEdit_, ScreenColor=(R=1,G=1,B=1,A=1), ScreenDuration=0.05, Value=@n13.Result) {template="DISPATCH_RigVMDispatch_Print(in Prefix,in Value,in Enabled,in ScreenDuration,in ScreenColor)"}
   > &RigUnit_ModifyTransforms(
    ItemToModify=((Item=(Type=Bone,Name="mstr_dyn_hairStrand_00_r"),Transform=(Rotation=(X=0,Y=0,Z=0,W=1),Translation=(X=0,Y=-0.5121,Z=1.0522),Scale3D=(X=1,Y=1,Z=1))))
    WeightMinimum=0
    Weight=@n13.Result
   )
   > n14 = &RigUnit_SphericalPoseReader(
    ActiveRegionScaleFactors=(PositiveWidth=1.2,NegativeWidth=1.01,PositiveHeight=1.2,NegativeHeight=1.3)
    ActiveRegionSize=0.05
    Debug=(bDrawDebug=true,bDraw2D=false,bDrawLocalAxes=false,DebugScale=25,DebugSegments=20,DebugThickness=0.25)
    DriverAxis=(X=0,Y=-1,Z=0)
    DriverItem=(Type=Bone,Name="head")
    FalloffRegionScaleFactors=(PositiveWidth=0.1,NegativeWidth=0.2,PositiveHeight=0.1,NegativeHeight=0.4)
    FalloffSize=1
    FlipHeightScaling=false
    FlipWidthScaling=false
    OptionalParentItem=(Type=Bone,Name="spine_05")
    RotationOffset=(X=60.3253,Y=71.7812,Z=97.316)
   )
   > n15 = &RigUnit_SphericalPoseReader(
    ActiveRegionScaleFactors=(PositiveWidth=1.2,NegativeWidth=1.01,PositiveHeight=1.2,NegativeHeight=1.3)
    ActiveRegionSize=0.1
    Debug=(bDrawDebug=true,bDraw2D=false,bDrawLocalAxes=false,DebugScale=25,DebugSegments=20,DebugThickness=0.25)
    DriverAxis=(X=0,Y=-1,Z=0)
    DriverItem=(Type=Bone,Name="head")
    FalloffRegionScaleFactors=(PositiveWidth=0.1,NegativeWidth=0.2,PositiveHeight=0.1,NegativeHeight=0.4)
    FalloffSize=1
    FlipHeightScaling=false
    FlipWidthScaling=false
    OptionalParentItem=(Type=Bone,Name="spine_05")
    RotationOffset=(X=60.3253,Y=71.7812,Z=65.5187)
   )
   n16 = RigVMVariableNode() {variable=IsCrouching}
   n17 = &RigVMFunction_AlphaInterp(
    InterpSpeedDecreasing=1
    InterpSpeedIncreasing=1
    Scale=1
    bClampResult=False
    Value=RigVMDispatchNode(Values=(0,1), Index=&RigVMFunction_MathBoolToInteger(Value=@n16.Value).Result) {template="DISPATCH_RigVMDispatch_SelectInt32(in Index,in Values,out Result)"}.Result
   )
   > RigVMVariableNode(Value=@n17.Result) {variable=IsCrouchingAlpha}
   Add = &RigVMFunction_MathDoubleAdd(A=@n14.OutputParam, B=@n15.OutputParam)
   n18 = &RigVMFunction_AlphaInterp(
    InRange=(Min=1,Max=2)
    InterpSpeedDecreasing=10
    InterpSpeedIncreasing=10
    Scale=1
    bClampResult=true
    bInterpResult=false
    bMapRange=true
    Value=@Add.Result
   )
   n19 = &RigVMFunction_MathDoubleMul(B=1, A=&RigVMFunction_MathDoubleMul(B=@n18.Result, A=RigVMVariableNode() {variable=IsCrouchingAlpha}.Value).Result)
   > RigVMDispatchNode(Enabled=false, Prefix=Crouch_, ScreenColor=(R=1,G=1,B=1,A=1), ScreenDuration=0.05, Value=@n19.Result) {template="DISPATCH_RigVMDispatch_Print(in Prefix,in Value,in Enabled,in ScreenDuration,in ScreenColor)"}
   > RigVMDispatchNode(
    Enabled=false
    Prefix=FIX_
    ScreenColor=(R=1,G=1,B=1,A=1)
    ScreenDuration=0.05
    Value=&RigVMFunction_MathDoubleMul(B=1, A=@n19.Result).Result
   ) {template="DISPATCH_RigVMDispatch_Print(in Prefix,in Value,in Enabled,in ScreenDuration,in ScreenColor)"}
   > &RigUnit_ModifyTransforms(
    ItemToModify=((Item=(Type=Bone,Name="mstr_dyn_hairBack_00"),Transform=(Rotation=(X=0,Y=-0.1736,Z=0,W=0.9848),Translation=(X=0,Y=0,Z=0),Scale3D=(X=1,Y=1,Z=1))))
    WeightMinimum=0
    Weight=@n19.Result
   )
   > &RigUnit_ModifyTransforms(
    ItemToModify=((Item=(Type=Bone,Name="mstr_dyn_hair_00_r"),Transform=(Rotation=(X=0,Y=0,Z=0,W=1),Translation=(X=0,Y=0.2184,Z=-0.3709),Scale3D=(X=1,Y=1,Z=1))))
    WeightMinimum=0
    Weight=@n19.Result
   )
   > &RigUnit_ModifyTransforms(
    ItemToModify=((Item=(Type=Bone,Name="mstr_dyn_hair_00_l"),Transform=(Rotation=(X=0,Y=0,Z=-0.0436,W=0.999),Translation=(X=-1.1474,Y=-0.1014,Z=-0.025),Scale3D=(X=1,Y=1,Z=1))))
    WeightMinimum=0
    Weight=@n19.Result
   )
   > &RigUnit_ModifyTransforms(
    ItemToModify=((Item=(Type=Bone,Name="mstr_dyn_hairBack_00"),Transform=(Rotation=(X=-0.0034,Y=-0.1302,Z=-0.0048,W=0.9915),Translation=(X=-2.5681,Y=-0.1141,Z=2.0452),Scale3D=(X=1,Y=1,Z=1))))
    Weight=0
    WeightMinimum=0
   )
   n20 = RigVMVariableNode() {variable=IsCrouchingAlpha}
   > &RigUnit_ModifyTransforms(
    ItemToModify=((Item=(Type=Bone,Name="mstr_dyn_hairBack_00"),Transform=(Rotation=(X=0,Y=-0.0872,Z=0,W=0.9962),Translation=(X=0,Y=0,Z=0),Scale3D=(X=1,Y=1,Z=1))))
    WeightMinimum=0
    Weight=@n20.Value
   )
   > &RigUnit_ModifyTransforms(
    ItemToModify=((Item=(Type=Bone,Name="mstr_dyn_hair_00_l"),Transform=(Rotation=(X=0,Y=0,Z=-0.0436,W=0.999),Translation=(X=-0.5,Y=0,Z=0),Scale3D=(X=1,Y=1,Z=1))))
    WeightMinimum=0
    Weight=@n20.Value
   )
   > &RigUnit_ModifyTransforms(
    ItemToModify=((Item=(Type=Bone,Name="mstr_dyn_hair_00_l"),Transform=(Rotation=(X=0,Y=0,Z=0,W=1),Translation=(X=-1,Y=0,Z=0),Scale3D=(X=1,Y=1,Z=1))))
    WeightMinimum=0
    Weight=RigVMVariableNode() {variable=IsCrouchingAlpha}.Value
   )
   > RigVMVariableNode(Value=@n19.Result) {variable=LookUpAndCrouchAlpha}
   > RigVMDispatchNode(
    Enabled=false
    Prefix=TEST
    ScreenColor=(R=1,G=1,B=1,A=1)
    ScreenDuration=0.05
    Value=RigVMVariableNode() {variable=LookUpAndCrouchAlpha}.Value
   ) {template="DISPATCH_RigVMDispatch_Print(in Prefix,in Value,in Enabled,in ScreenDuration,in ScreenColor)"}
   > &RigUnit_SphericalPoseReader(
    ActiveRegionScaleFactors=(PositiveWidth=1.2,NegativeWidth=1.01,PositiveHeight=1.2,NegativeHeight=1.3)
    ActiveRegionSize=0.1
    Debug=(bDrawDebug=false,bDraw2D=false,bDrawLocalAxes=false,DebugScale=25,DebugSegments=20,DebugThickness=0.25)
    DriverAxis=(X=0,Y=-1,Z=0)
    DriverItem=(Type=Bone,Name="head")
    FalloffRegionScaleFactors=(PositiveWidth=0.1,NegativeWidth=0.2,PositiveHeight=0.1,NegativeHeight=0.4)
    FalloffSize=1
    FlipHeightScaling=false
    FlipWidthScaling=false
    OptionalParentItem=(Type=Bone,Name="spine_05")
    RotationOffset=(X=60.3253,Y=71.7812,Z=65.5187)
   )
   n21 = RigVMVariableNode() {variable=isTargetingAlpha}
   n22 = &RigVMFunction_MathDoubleMul(A=&RigVMFunction_MathDoubleMul(B=1, A=@n18.Result).Result, B=@n21.Value)
   > RigVMDispatchNode(Enabled=false, Prefix=CrouchTarget, ScreenColor=(R=1,G=1,B=1,A=1), ScreenDuration=0.05, Value=@n22.Result) {template="DISPATCH_RigVMDispatch_Print(in Prefix,in Value,in Enabled,in ScreenDuration,in ScreenColor)"}
   RigVMFunction_MathBoolToInteger = &RigVMFunction_MathBoolToInteger(Value=RigVMVariableNode() {variable=IsMoving2d}.Value)
   n23 = &RigVMFunction_AlphaInterp(
    InterpSpeedDecreasing=10
    InterpSpeedIncreasing=10
    Scale=1
    bClampResult=false
    Value=RigVMDispatchNode(Values=(0,1), Index=@RigVMFunction_MathBoolToInteger.Result) {template="DISPATCH_RigVMDispatch_SelectInt32(in Index,in Values,out Result)"}.Result
   )
   n24 = &RigVMFunction_MathDoubleMul(B=@n23.Result, A=@n21.Value)
   > &RigUnit_ModifyTransforms(
    ItemToModify=((Item=(Type=Bone,Name="mstr_dyn_hairBack_00"),Transform=(Rotation=(X=0,Y=-0.2588,Z=0,W=0.9659),Translation=(X=0,Y=0,Z=1.0943),Scale3D=(X=1,Y=1,Z=1))))
    WeightMinimum=0
    Weight=@n24.Result
   )
   > &RigUnit_ModifyTransforms(
    ItemToModify=((Item=(Type=Bone,Name="mstr_dyn_hair_00_r"),Transform=(Rotation=(X=0,Y=-0.1219,Z=0,W=0.9925),Translation=(X=0,Y=0,Z=1.0943),Scale3D=(X=1,Y=1,Z=1))))
    WeightMinimum=0
    Weight=@n22.Result
   )
   > &RigUnit_ModifyTransforms(
    ItemToModify=((Item=(Type=Bone,Name="dyn_hair_00_l"),Transform=(Rotation=(X=0,Y=0,Z=0,W=1),Translation=(X=2.4806,Y=-1.5724,Z=0),Scale3D=(X=1,Y=1,Z=1))))
    WeightMinimum=0
    Weight=@n22.Result
   )
   > &RigUnit_ModifyTransforms(
    ItemToModify=((Item=(Type=Bone,Name="dyn_hairBack_01"),Transform=(Rotation=(X=0,Y=-0.1305,Z=0,W=0.9914),Translation=(X=0,Y=0,Z=0),Scale3D=(X=1,Y=1,Z=1))))
    WeightMinimum=0
    Weight=@n22.Result
   )
   > RigVMDispatchNode(Enabled=false, Prefix=CrouchTargetFIX_, ScreenColor=(R=1,G=1,B=1,A=1), ScreenDuration=0.05, Value=@n24.Result) {template="DISPATCH_RigVMDispatch_Print(in Prefix,in Value,in Enabled,in ScreenDuration,in ScreenColor)"}
   > &RigUnit_ModifyTransforms(
    ItemToModify=((Item=(Type=Bone,Name="mstr_dyn_hairBack_00"),Transform=(Rotation=(X=0,Y=0.5,Z=0,W=0.866),Translation=(X=0,Y=1,Z=-1),Scale3D=(X=1,Y=1,Z=1))))
    WeightMinimum=0
    Weight=@n24.Result
   )
   > &RigUnit_ModifyTransforms(
    ItemToModify=((Item=(Type=Bone,Name="mstr_dyn_hair_00_r"),Transform=(Rotation=(X=0,Y=0,Z=0,W=1),Translation=(X=0.109,Y=0.7255,Z=-0.5482),Scale3D=(X=1,Y=1,Z=1))))
    WeightMinimum=0
    Weight=@n24.Result
   )
   &RigVMFunction_MathDoubleMul(B=@n1.OutputParam, A=RigVMVariableNode() {variable=IsMoving2dAlpha}.Value)
   RigVMVariableNode() {variable=IsCrouchingAlpha}
   RigVMVariableNode() {variable=IsMoving2dAlpha}
   RigVMVariableNode() {variable=IsCrouching}
   RigVMVariableNode() {variable=IsNotCrouchingAlpha}
   RigVMVariableNode() {variable=IsMoving2dAlpha}
  ]}
]
hierarchy [
 root {loc=<0 0 0>, rot=<0 0 0 1>}
 pelvis {parent=root, loc=<0 -0.7071 93.1191>, rot=<0.0127 0.707 -0.0127 -0.707>}
 spine_01 {parent=pelvis, loc=<6.7494 -0.6388 0>, rot=<0 0 0.034 -0.9994>}
 spine_02 {parent=spine_01, loc=<7.2518 -0.035 0>, rot=<0 0 0.0238 -0.9997>}
 spine_03 {parent=spine_02, loc=<6.0104 0 0>, rot=<0 0 0.0149 -0.9999>}
 spine_04 {parent=spine_03, loc=<8.1024 0 0>, rot=<0 0 -0.0132 -0.9999>}
 spine_05 {parent=spine_04, loc=<10.0146 0 0>, rot=<0 0 -0.0015 -1>}
 neck_01 {parent=spine_05, loc=<11.3138 0.1529 0>, rot=<0 0 -0.2003 -0.9797>}
 neck_02 {parent=neck_01, loc=<4.5627 0 0>, rot=<0 0 0.0504 -0.9987>}
 head {parent=neck_02, loc=<5.3914 0 0>, rot=<0 0 0.115 -0.9934>}
 mstr_dyn_hairBang2_00_l {parent=head, loc=<9.8078 7.4034 -3.8956>, rot=<0.8066 0.5386 -0.1256 0.2086>}
 dyn_hairBan2_00_l {parent=mstr_dyn_hairBang2_00_l, loc=<0 0 0>, rot=<0 0 -0.0829 -0.9966>}
 dyn_hairBang2_01_l {parent=dyn_hairBan2_00_l, loc=<2.7265 0.0733 0.0594>, rot=<0.0821 0.0168 0.8704 -0.4852>}
 dyn_hairBang2_02_l {parent=dyn_hairBang2_01_l, loc=<4.336 -0.1083 0.054>, rot=<0.0645 0.233 -0.2188 -0.9454>}
 dyn_hairBang2_03_l {parent=dyn_hairBang2_02_l, loc=<2.8836 0.0746 0.0921>, rot=<-0.0217 -0.2949 0.1445 -0.9443>}
 FACIAL_C_FacialRoot {parent=head, loc=<2.6326 1.5213 0>, rot=<0.4832 -0.5162 -0.4832 -0.5163>}
 FACIAL_C_Forehead {parent=FACIAL_C_FacialRoot, loc=<0.0002 -2.443 7.8342>, rot=<-0.0403 0.0001 0 -0.9992>}
 FACIAL_L_ForeheadIn {parent=FACIAL_C_FacialRoot, loc=<1.5235 -2.5263 7.8013>, rot=<-0.0575 -0.0695 0.0045 -0.9959>}
 FACIAL_R_ForeheadIn {parent=FACIAL_C_FacialRoot, loc=<-1.4951 -2.562 7.7504>, rot=<-0.058 0.0696 -0.0045 -0.9959>}
 FACIAL_L_ForeheadMid {parent=FACIAL_C_FacialRoot, loc=<2.8052 -2.8412 7.2846>, rot=<-0.0681 -0.1755 -0.02 -0.9819>}
 FACIAL_R_ForeheadMid {parent=FACIAL_C_FacialRoot, loc=<-2.7694 -2.8753 7.2299>, rot=<-0.0686 0.1756 0.0201 -0.9819>}
 FACIAL_L_ForeheadOut {parent=FACIAL_C_FacialRoot, loc=<3.9823 -2.8723 6.4042>, rot=<-0.1182 -0.2556 -0.0782 -0.9563>}
 FACIAL_R_ForeheadOut {parent=FACIAL_C_FacialRoot, loc=<-3.9423 -2.9069 6.3645>, rot=<-0.1188 0.2556 0.0783 -0.9563>}
 FACIAL_L_EyesackUpper {parent=FACIAL_C_FacialRoot, loc=<2.9309 -1.6802 7.6249>, rot=<0.0553 -0.0846 -0.0639 -0.9928>}
 FACIAL_R_EyesackUpper {parent=FACIAL_C_FacialRoot, loc=<-2.9095 -1.6665 7.6368>, rot=<0.0549 0.0846 0.0639 -0.9928>}
 FACIAL_L_EyelidUpperFurrow {parent=FACIAL_C_FacialRoot, loc=<2.762 -1.5716 7.4391>, rot=<-0.0199 -0.1516 -0.0781 -0.9852>}
 FACIAL_R_EyelidUpperFurrow {parent=FACIAL_C_FacialRoot, loc=<-2.7626 -1.5615 7.4374>, rot=<-0.0202 0.1516 0.0781 -0.9852>}
 FACIAL_L_EyelidUpperA {parent=FACIAL_C_FacialRoot, loc=<2.7321 -1.1096 5.8732>, rot=<-0.1326 0 0 -0.9912>}
 FACIAL_R_EyelidUpperA {parent=FACIAL_C_FacialRoot, loc=<-2.7278 -1.1127 5.858>, rot=<-0.1327 0 0 -0.9912>}
 FACIAL_L_Eye {parent=FACIAL_C_FacialRoot, loc=<2.7321 -1.1096 5.8732>, rot=<0.0154 0 0 -0.9999>}
 FACIAL_R_Eye {parent=FACIAL_C_FacialRoot, loc=<-2.7276 -1.1111 5.8667>, rot=<0.0152 0 0 -0.9999>}
 FACIAL_L_EyelidLowerA {parent=FACIAL_C_FacialRoot, loc=<2.7321 -1.1096 5.8732>, rot=<0.2987 0 0 -0.9543>}
 FACIAL_R_EyelidLowerA {parent=FACIAL_C_FacialRoot, loc=<-2.7278 -1.1127 5.858>, rot=<0.2986 0 0 -0.9544>}
 FACIAL_L_EyeCornerInner {parent=FACIAL_C_FacialRoot, loc=<1.8108 -0.7387 6.8758>, rot=<0.0154 0 0 -0.9999>}
 FACIAL_R_EyeCornerInner {parent=FACIAL_C_FacialRoot, loc=<-1.8134 -0.7411 6.8734>, rot=<0.0152 0 0 -0.9999>}
 FACIAL_L_EyeCornerOuter {parent=FACIAL_C_FacialRoot, loc=<4.1566 -1.1552 6.1656>, rot=<0.0142 -0.3826 0.0059 -0.9238>}
 FACIAL_R_EyeCornerOuter {parent=FACIAL_C_FacialRoot, loc=<-4.1486 -1.1482 6.1581>, rot=<0.0141 0.3826 -0.0058 -0.9238>}
 FACIAL_L_EyesackLower {parent=FACIAL_C_FacialRoot, loc=<2.9555 0.1029 6.0934>, rot=<-0.007 -0.1401 -0.0201 -0.9899>}
 FACIAL_R_EyesackLower {parent=FACIAL_C_FacialRoot, loc=<-2.9579 0.1067 6.0836>, rot=<-0.0072 0.14 0.0201 -0.9899>}
 FACIAL_L_CheekInner {parent=FACIAL_C_FacialRoot, loc=<2.0016 0.8501 6.461>, rot=<-0.109 -0.0909 -0.1959 -0.9703>}
 FACIAL_R_CheekInner {parent=FACIAL_C_FacialRoot, loc=<-2.0011 0.857 6.459>, rot=<-0.1091 0.0908 0.1958 -0.9703>}
 FACIAL_L_CheekOuter {parent=FACIAL_C_FacialRoot, loc=<3.5027 0.9736 4.8625>, rot=<0.0694 -0.3754 0.106 -0.9182>}
 FACIAL_R_CheekOuter {parent=FACIAL_C_FacialRoot, loc=<-3.4908 0.9742 4.846>, rot=<0.0692 0.3753 -0.1063 -0.9182>}
 FACIAL_L_NasolabialBulge {parent=FACIAL_C_FacialRoot, loc=<3.2692 4.3453 4.4852>, rot=<0.069 -0.3771 0.0283 -0.9232>}
 FACIAL_R_NasolabialBulge {parent=FACIAL_C_FacialRoot, loc=<-3.2659 4.3367 4.4824>, rot=<0.0689 0.3771 -0.0284 -0.9232>}
 FACIAL_L_NasolabialFurrow {parent=FACIAL_C_FacialRoot, loc=<2.8565 3.0396 6.8003>, rot=<-0.0202 -0.131 -0.362 -0.9227>}
 FACIAL_R_NasolabialFurrow {parent=FACIAL_C_FacialRoot, loc=<-2.8476 3.057 6.8063>, rot=<-0.0202 0.1309 0.3618 -0.9228>}
 FACIAL_L_CheekLower {parent=FACIAL_C_FacialRoot, loc=<3.678 4.0251 3.8277>, rot=<0.0454 -0.4262 0.0628 -0.9013>}
 FACIAL_R_CheekLower {parent=FACIAL_C_FacialRoot, loc=<-3.6752 4.0153 3.8279>, rot=<0.0453 0.4262 -0.0629 -0.9013>}
 FACIAL_C_Nose {parent=FACIAL_C_FacialRoot, loc=<0 0.6977 8.3588>, rot=<-0.1166 0 -0.0001 -0.9932>}
 FACIAL_C_NoseLower {parent=FACIAL_C_Nose, loc=<0.0002 2.8002 -0.6755>, rot=<0.1302 0 0 -0.9915>}
 FACIAL_L_Nostril {parent=FACIAL_C_Nose, loc=<1.0068 2.0154 -0.7116>, rot=<0.0173 0.1305 0.0023 -0.9913>}
 FACIAL_R_Nostril {parent=FACIAL_C_Nose, loc=<-1.0114 2.019 -0.7055>, rot=<0.0173 -0.1305 -0.0023 -0.9913>}
 FACIAL_C_NoseTip {parent=FACIAL_C_Nose, loc=<-0.0007 2.0669 0.3964>, rot=<0.0872 0 0 -0.9962>}
 FACIAL_C_MouthUpper {parent=FACIAL_C_FacialRoot, loc=<0.0005 4.4963 5.464>, rot=<0.012 0 0 -0.9999>}
 FACIAL_C_LipUpper {parent=FACIAL_C_MouthUpper, loc=<0 -0.0666 2.1282>, rot=<-0.0872 0 0 -0.9962>}
 FACIAL_L_LipUpper {parent=FACIAL_C_MouthUpper, loc=<1.1266 -0.0093 1.8782>, rot=<-0.0778 -0.1601 -0.0379 -0.9833>}
 FACIAL_R_LipUpper {parent=FACIAL_C_MouthUpper, loc=<-1.1287 -0.0084 1.8784>, rot=<-0.0778 0.1601 0.0379 -0.9833>}
 FACIAL_L_LipUpperOuter {parent=FACIAL_C_MouthUpper, loc=<1.8613 0.135 1.2905>, rot=<-0.0983 -0.2294 -0.0983 -0.9634>}
 FACIAL_R_LipUpperOuter {parent=FACIAL_C_MouthUpper, loc=<-1.8659 0.133 1.2913>, rot=<-0.0983 0.2294 0.0983 -0.9634>}
 FACIAL_L_LipCorner {parent=FACIAL_C_MouthUpper, loc=<2.3471 0.3983 0.621>, rot=<0 -0.2079 0 -0.9782>}
 FACIAL_R_LipCorner {parent=FACIAL_C_MouthUpper, loc=<-2.353 0.3987 0.6193>, rot=<0 0.2079 0 -0.9782>}
 FACIAL_L_Masseter {parent=FACIAL_C_FacialRoot, loc=<5.5335 1.4638 2.8605>, rot=<0.0862 0.1129 0.0059 -0.9898>}
 FACIAL_R_Masseter {parent=FACIAL_C_FacialRoot, loc=<-5.526 1.4729 2.8573>, rot=<0.0858 -0.1131 -0.0061 -0.9898>}
 FACIAL_C_UnderChin {parent=FACIAL_C_FacialRoot, loc=<-0.0004 8.8994 3.0616>, rot=<0.503 -0.0001 -0.0001 -0.8643>}
 FACIAL_L_UnderChin {parent=FACIAL_C_FacialRoot, loc=<3.1712 7.1905 1.6473>, rot=<0.4103 -0.3541 0.0852 -0.8361>}
 FACIAL_R_UnderChin {parent=FACIAL_C_FacialRoot, loc=<-3.141 7.1714 1.6642>, rot=<0.4094 0.3537 -0.0853 -0.8367>}
 FACIAL_C_TeethUpper {parent=FACIAL_C_FacialRoot, loc=<-0.0002 4.9822 6.834>, rot=<0.012 0 0 -0.9999>}
 FACIAL_C_LowerLipRotation {parent=FACIAL_C_FacialRoot, loc=<-0.0001 2.0644 1.2027>, rot=<0.3533 0 0 -0.9355>}
 FACIAL_C_MouthLower {parent=FACIAL_C_LowerLipRotation, loc=<0.0007 -0.9916 4.8052>, rot=<-0.342 0 0 -0.9397>}
 FACIAL_C_LipLower {parent=FACIAL_C_MouthLower, loc=<-0.001 1.3312 1.8131>, rot=<0.0872 0 0 -0.9962>}
 FACIAL_L_LipLower {parent=FACIAL_C_MouthLower, loc=<1.0106 1.2922 1.6503>, rot=<0.0927 -0.1295 0.0273 -0.9869>}
 FACIAL_R_LipLower {parent=FACIAL_C_MouthLower, loc=<-1.0136 1.2908 1.6496>, rot=<0.0927 0.1295 -0.0273 -0.9869>}
 FACIAL_L_LipLowerOuter {parent=FACIAL_C_MouthLower, loc=<1.8299 0.8821 1.1294>, rot=<0.0913 -0.2322 0.0318 -0.9679>}
 FACIAL_R_LipLowerOuter {parent=FACIAL_C_MouthLower, loc=<-1.837 0.886 1.1289>, rot=<0.0913 0.2322 -0.0318 -0.9679>}
 FACIAL_C_TeethLower {parent=FACIAL_C_LowerLipRotation, loc=<-0.0001 -1.4685 6.0974>, rot=<-0.342 0 0 -0.9397>}
 FACIAL_C_Tongue1 {parent=FACIAL_C_TeethLower, loc=<0.0002 -0.0229 -5.0745>, rot=<-0.2588 0 0 -0.9659>}
 FACIAL_C_Tongue2 {parent=FACIAL_C_Tongue1, loc=<-0.0001 0.4356 1.4758>, rot=<0.2588 0 0 -0.9659>}
 FACIAL_C_Tongue3 {parent=FACIAL_C_Tongue2, loc=<0 0.2235 1.3328>, rot=<0.1305 0 0 -0.9914>}
 FACIAL_C_Tongue4 {parent=FACIAL_C_Tongue3, loc=<0.0006 -0.1727 1.127>, rot=<0.0436 0 0 -0.999>}
 FACIAL_C_TongueUpper3 {parent=FACIAL_C_Tongue3, loc=<-0.0001 -0.3748 -0.1408>, rot=<0 0 0 1>}
 FACIAL_C_TongueLower3 {parent=FACIAL_C_Tongue3, loc=<0.0004 0.8101 0.1038>, rot=<0 0 0 1>}
 FACIAL_L_TongueSide3 {parent=FACIAL_C_Tongue3, loc=<1.5612 0.0549 0.1827>, rot=<0 0 0 1>}
 FACIAL_R_TongueSide3 {parent=FACIAL_C_Tongue3, loc=<-1.5597 0.0506 0.1784>, rot=<0 0 0 1>}
 FACIAL_C_TongueUpper2 {parent=FACIAL_C_Tongue2, loc=<0.0004 -0.5585 0.0123>, rot=<0 0 0 1>}
 FACIAL_L_TongueSide2 {parent=FACIAL_C_Tongue2, loc=<1.8007 0.1686 0.3858>, rot=<0 0 0 1>}
 FACIAL_R_TongueSide2 {parent=FACIAL_C_Tongue2, loc=<-1.7972 0.1622 0.382>, rot=<0 0 0 1>}
 FACIAL_C_Jaw {parent=FACIAL_C_FacialRoot, loc=<-0.0001 2.0644 1.2027>, rot=<0.3533 0 0 -0.9355>}
 FACIAL_C_Jawline {parent=FACIAL_C_Jaw, loc=<0.0001 2.9492 7.5925>, rot=<0.2549 0 0 -0.967>}
 FACIAL_L_ChinSide {parent=FACIAL_C_Jaw, loc=<2.8475 0.8739 6.7998>, rot=<-0.2197 -0.3737 0.0503 -0.8998>}
 FACIAL_R_ChinSide {parent=FACIAL_C_Jaw, loc=<-2.7946 0.8656 6.772>, rot=<-0.22 0.3736 -0.0502 -0.8997>}
 FACIAL_C_Chin {parent=FACIAL_C_Jaw, loc=<0 1.5566 7.6622>, rot=<-0.26 0 0 -0.9656>}
 FACIAL_L_Ear {parent=FACIAL_C_FacialRoot, loc=<6.517 -0.2349 0.0254>, rot=<-0.0081 -0.0012 -0.0261 -0.9996>}
 FACIAL_R_Ear {parent=FACIAL_C_FacialRoot, loc=<-6.5224 -0.2323 0.0155>, rot=<-0.0085 0.0011 0.0261 -0.9996>}
 FACIAL_L_Jawline {parent=FACIAL_C_FacialRoot, loc=<3.8573 5.4456 3.0574>, rot=<0.2237 0.1565 -0.0571 -0.9603>}
 FACIAL_R_Jawline {parent=FACIAL_C_FacialRoot, loc=<-3.8129 5.3954 3.0431>, rot=<0.2224 -0.1573 0.0566 -0.9605>}
 FACIAL_C_NoseBridge {parent=FACIAL_C_FacialRoot, loc=<0.0001 -1.1534 8.5114>, rot=<0.0477 0.0001 0 -0.9989>}
 FACIAL_L_EyelidUpperB {parent=FACIAL_C_FacialRoot, loc=<2.7321 -1.1096 5.8732>, rot=<-0.2099 0 0 -0.9777>}
 FACIAL_L_EyelidLowerB {parent=FACIAL_C_FacialRoot, loc=<2.7321 -1.1096 5.8732>, rot=<0.3319 0 0 -0.9433>}
 FACIAL_R_EyelidLowerB {parent=FACIAL_C_FacialRoot, loc=<-2.7278 -1.1127 5.858>, rot=<0.3399 0 0 -0.9405>}
 FACIAL_R_EyelidUpperB {parent=FACIAL_C_FacialRoot, loc=<-2.7278 -1.1127 5.858>, rot=<-0.2101 0 0 -0.9777>}
 mstr_dyn_hair_00_r {parent=head, loc=<8.968 -3.6057 6.0477>, rot=<0.0201 -0.2622 0.9621 -0.0718>}
 dyn_hair_00_r {parent=mstr_dyn_hair_00_r, loc=<0 0 0>, rot=<0 0 0 1>}
 dyn_hair_cstr_r {parent=dyn_hair_00_r, loc=<0 0 0>, rot=<0 0 0 1>}
 dyn_hair_01_00_r {parent=dyn_hair_cstr_r, loc=<7.2649 -0.5058 0.1369>, rot=<0 0 0.0696 -0.9976>}
 mstr_dyn_hairStrand_00_r {parent=head, loc=<7.342 -0.9635 8.1595>, rot=<0.3535 0.169 0.8603 -0.3261>}
 dyn_hairStrand_00_r {parent=mstr_dyn_hairStrand_00_r, loc=<0 0 0>, rot=<0 -0.2663 0 -0.9639>}
 dyn_hairStrand_01_r {parent=dyn_hairStrand_00_r, loc=<1.3453 0.2687 0.6531>, rot=<0 0 0 1>}
 mstr_dyn_hairBang_00_r {parent=head, loc=<9.7254 9.1629 0.6091>, rot=<-0.2086 0.1256 -0.5386 -0.8066>}
 dyn_hairBang_00_r {parent=mstr_dyn_hairBang_00_r, loc=<0 0 0>, rot=<0 0 0 1>}
 dyn_hairBang_01_r {parent=dyn_hairBang_00_r, loc=<3.3176 -1.4595 0.9689>, rot=<-0.0787 0.0323 -0.8451 -0.5278>}
 dyn_hairBang_02_00_r {parent=dyn_hairBang_01_r, loc=<4.0541 -1.3144 0.0973>, rot=<-0.0501 -0.3382 0.0508 -0.9383>}
 dyn_hairBang_03_00_r {parent=dyn_hairBang_02_00_r, loc=<2.334 0.9891 0.2137>, rot=<-0.0589 0.2831 0.035 -0.9566>}
 mstr_dyn_hair_00_l {parent=head, loc=<8.9677 -3.6057 -6.0477>, rot=<0.0718 -0.9621 0.2622 -0.0201>}
 dyn_hair_00_l {parent=mstr_dyn_hair_00_l, loc=<0 0 0>, rot=<0 0 0 1>}
 dyn_hair_cstr_l {parent=dyn_hair_00_l, loc=<0 0 0>, rot=<0 0 0 1>}
 dyn_hair_01_00_l {parent=dyn_hair_cstr_l, loc=<7.2643 0.5057 0.1369>, rot=<0 0 -0.0696 -0.9976>}
 mstr_dyn_hairBack_00 {parent=head, loc=<9.9057 -4.8874 -0.3969>, rot=<0.1496 -0.6911 0.6911 -0.1496>}
 dyn_hairBack_00 {parent=mstr_dyn_hairBack_00, loc=<0 0 0>, rot=<0 0 0 1>}
 dyn_hairBack_cstr {parent=dyn_hairBack_00, loc=<0 0 0>, rot=<-0.0062 -0.0286 0.0042 -0.9996>}
 dyn_hairBack_01 {parent=dyn_hairBack_00, loc=<8.4928 -0.0175 -1.016>, rot=<0 -0.0946 0 -0.9955>}
 mstr_dyn_hairStrand_00_l {parent=head, loc=<7.3418 -0.9635 -8.1595>, rot=<0.3593 -0.9234 -0.076 -0.1116>}
 dyn_hairStrand_00_l {parent=mstr_dyn_hairStrand_00_l, loc=<0 0 0>, rot=<0 0 0 1>}
 dyn_hairStrand_01_l {parent=dyn_hairStrand_00_l, loc=<2.8695 0 0>, rot=<0 0 0 1>}
 mstr_dyn_hairBang_00_l {parent=head, loc=<9.7255 9.1629 -0.6091>, rot=<0.8066 0.5386 -0.1256 0.2086>}
 dyn_hairBang_00_l {parent=mstr_dyn_hairBang_00_l, loc=<0 0 0>, rot=<0 0 0 1>}
 dyn_hairBang_01_l {parent=dyn_hairBang_00_l, loc=<3.8001 -0.0007 0.031>, rot=<0.1464 0.0799 0.8305 -0.5315>}
 dyn_hairBang_02_00_l {parent=dyn_hairBang_01_l, loc=<6.4651 -0.266 0.0015>, rot=<0.0717 -0.1403 -0.1121 -0.9811>}
 dyn_hairBang_03_00_l {parent=dyn_hairBang_02_00_l, loc=<3.8766 0 0>, rot=<0 0 0 1>}
 clavicle_l {parent=spine_05, loc=<4.3575 -0.1466 -2.3223>, rot=<0.0465 -0.7289 -0.009 -0.6829>}
 upperarm_l {parent=clavicle_l, loc=<15.5111 0 0>, rot=<0.0226 -0.3234 -0.0088 -0.9459>}
 lowerarm_l {parent=upperarm_l, loc=<26.6514 0 0>, rot=<0.0391 -0.0099 -0.2453 -0.9686>}
 hand_l {parent=lowerarm_l, loc=<22.3491 0 0>, rot=<-0.7081 -0.0507 0.0975 -0.6975>}
 thumb_01_l {parent=hand_l, loc=<1.3118 -0.3191 -2.459>, rot=<0.6546 -0.4042 0.0664 -0.6354>}
 thumb_02_l {parent=thumb_01_l, loc=<5.9336 -0.0448 0.0401>, rot=<-0.0037 0.0213 0.298 -0.9543>}
 thumb_03_l {parent=thumb_02_l, loc=<2.2223 0.3682 -0.0907>, rot=<0.0118 -0.013 0.0066 -0.9998>}
 index_metacarpal_l {parent=hand_l, loc=<1.6535 0.0183 -2.96>, rot=<0.0158 -0.0124 -0.0173 -0.9996>}
 index_01_l {parent=index_metacarpal_l, loc=<8.3008 0.0357 -0.0524>, rot=<-0.0044 0.0485 0.0924 -0.9945>}
 index_02_l {parent=index_01_l, loc=<3.3587 -0.3004 0.0452>, rot=<-0.002 0.0027 0.3225 -0.9466>}
 index_03_l {parent=index_02_l, loc=<2.5578 -0.123 -0.0252>, rot=<-0.002 0.0067 0.2558 -0.9667>}
 middle_metacarpal_l {parent=hand_l, loc=<2.1856 -0.1725 -0.5984>, rot=<0.006 0.0168 -0.0329 -0.9993>}
 middle_01_l {parent=middle_metacarpal_l, loc=<7.3275 -0.2447 0.0129>, rot=<-0.0208 0.0042 0.0858 -0.9961>}
 middle_02_l {parent=middle_01_l, loc=<4.0123 0.1401 0.1175>, rot=<-0.0218 0.0258 0.3512 -0.9357>}
 middle_03_l {parent=middle_02_l, loc=<2.6077 0.0534 0.1373>, rot=<0.0082 0.0188 0.1708 -0.9851>}
 ring_metacarpal_l {parent=hand_l, loc=<2.2074 -0.3625 1.3638>, rot=<-0.0284 0.0722 -0.0458 -0.9959>}
 ring_01_l {parent=ring_metacarpal_l, loc=<7.0295 -0.6197 -0.1446>, rot=<-0.0489 -0.0252 0.1892 -0.9804>}
 ring_02_l {parent=ring_01_l, loc=<3.6174 0.4647 -0.1027>, rot=<0 0 0.2806 -0.9598>}
 ring_03_l {parent=ring_02_l, loc=<2.4152 0.114 0.0199>, rot=<0 0 0.1189 -0.9929>}
 pinky_metacarpal_l {parent=hand_l, loc=<1.7616 -0.9753 3.3814>, rot=<-0.0397 0.0814 -0.069 -0.9935>}
 pinky_01_l {parent=pinky_metacarpal_l, loc=<7.1709 -0.4831 -0.1094>, rot=<-0.0654 -0.0035 0.2242 -0.9723>}
 pinky_02_l {parent=pinky_01_l, loc=<2.8274 0.0893 -0.1043>, rot=<0.0037 0.0011 0.2854 -0.9584>}
 pinky_03_l {parent=pinky_02_l, loc=<2.2046 0.0307 -0.0078>, rot=<-0.0002 0.0098 0.1445 -0.9895>}
 lowerarm_twist_02_l {parent=lowerarm_l, loc=<9.6199 0 0>, rot=<0 0 0 -1>}
 lowerarm_twist_01_l {parent=lowerarm_l, loc=<17.8549 0 0>, rot=<0 0 0 -1>}
 upperarm_twist_01_l {parent=upperarm_l, loc=<0.0041 0 0>, rot=<0 0 0 1>}
 upperarm_twist_02_l {parent=upperarm_l, loc=<13.7941 0 0>, rot=<0 0 0 1>}
 clavicle_r {parent=spine_05, loc=<4.3575 -0.1466 2.3223>, rot=<0.7289 0.0465 -0.6829 0.009>}
 upperarm_r {parent=clavicle_r, loc=<-15.5111 0 0>, rot=<0.0226 -0.3234 -0.0088 -0.9459>}
 lowerarm_r {parent=upperarm_r, loc=<-26.6514 0 0>, rot=<0.0391 -0.0099 -0.2453 -0.9686>}
 hand_r {parent=lowerarm_r, loc=<-22.3491 0 0>, rot=<-0.7081 -0.0507 0.0975 -0.6975>}
 thumb_01_r {parent=hand_r, loc=<-1.3118 0.3191 2.459>, rot=<0.6546 -0.4042 0.0664 -0.6354>}
 thumb_02_r {parent=thumb_01_r, loc=<-5.9336 0.0448 -0.0401>, rot=<-0.0037 0.0213 0.298 -0.9543>}
 thumb_03_r {parent=thumb_02_r, loc=<-2.2223 -0.3682 0.0907>, rot=<0.0118 -0.013 0.0066 -0.9998>}
 index_metacarpal_r {parent=hand_r, loc=<-1.6535 -0.0183 2.96>, rot=<0.0158 -0.0124 -0.0173 -0.9996>}
 index_01_r {parent=index_metacarpal_r, loc=<-8.3008 -0.0357 0.0524>, rot=<-0.0044 0.0485 0.0924 -0.9945>}
 index_02_r {parent=index_01_r, loc=<-3.3587 0.3004 -0.0452>, rot=<-0.002 0.0027 0.3225 -0.9466>}
 index_03_r {parent=index_02_r, loc=<-2.5578 0.123 0.0252>, rot=<-0.002 0.0067 0.2558 -0.9667>}
 middle_metacarpal_r {parent=hand_r, loc=<-2.1856 0.1725 0.5984>, rot=<0.006 0.0168 -0.0329 -0.9993>}
 middle_01_r {parent=middle_metacarpal_r, loc=<-7.3275 0.2447 -0.0129>, rot=<-0.0208 0.0042 0.0858 -0.9961>}
 middle_02_r {parent=middle_01_r, loc=<-4.0123 -0.1401 -0.1175>, rot=<-0.0218 0.0258 0.3512 -0.9357>}
 middle_03_r {parent=middle_02_r, loc=<-2.6077 -0.0534 -0.1373>, rot=<0.0082 0.0188 0.1708 -0.9851>}
 ring_metacarpal_r {parent=hand_r, loc=<-2.2074 0.3625 -1.3638>, rot=<-0.0284 0.0722 -0.0458 -0.9959>}
 ring_01_r {parent=ring_metacarpal_r, loc=<-7.0295 0.6197 0.1446>, rot=<-0.0489 -0.0252 0.1892 -0.9804>}
 ring_02_r {parent=ring_01_r, loc=<-3.6174 -0.4647 0.1027>, rot=<0 0 0.2806 -0.9598>}
 ring_03_r {parent=ring_02_r, loc=<-2.4152 -0.114 -0.0199>, rot=<0 0 0.1189 -0.9929>}
 pinky_metacarpal_r {parent=hand_r, loc=<-1.7616 0.9753 -3.3814>, rot=<-0.0397 0.0814 -0.069 -0.9935>}
 pinky_01_r {parent=pinky_metacarpal_r, loc=<-7.1709 0.4831 0.1094>, rot=<-0.0654 -0.0035 0.2242 -0.9723>}
 pinky_02_r {parent=pinky_01_r, loc=<-2.8274 -0.0893 0.1043>, rot=<0.0037 0.0011 0.2854 -0.9584>}
 pinky_03_r {parent=pinky_02_r, loc=<-2.2046 -0.0307 0.0078>, rot=<-0.0002 0.0098 0.1445 -0.9895>}
 lowerarm_twist_02_r {parent=lowerarm_r, loc=<-9.6199 0 0>, rot=<0 0 0 -1>}
 lowerarm_twist_01_r {parent=lowerarm_r, loc=<-17.8549 0 0>, rot=<0 0 0 -1>}
 upperarm_twist_01_r {parent=upperarm_r, loc=<-0.0041 0 0>, rot=<0 0 0 -1>}
 upperarm_twist_02_r {parent=upperarm_r, loc=<-13.7941 0 0>, rot=<0 0 0 -1>}
 thigh_l {parent=pelvis, loc=<-2.5693 -0.0109 -7.7501>, rot=<0.0012 0.0572 0.0155 -0.9982>}
 calf_l {parent=thigh_l, loc=<-39.0484 0 0>, rot=<0.0006 -0.0161 -0.0391 -0.9991>}
 foot_l {parent=calf_l, loc=<-42.561 0.1077 0.4292>, rot=<0.0418 -0.0455 0.0343 -0.9975>}
 ball_l {parent=foot_l, loc=<-6.1599 13.8264 0.6806>, rot=<0.0001 -0.0001 -0.7063 -0.7079>}
 calf_twist_02_l {parent=calf_l, loc=<-18.8064 -0.0967 0.0397>, rot=<0.0237 0.0095 -0.0035 -0.9997>}
 calf_twist_01_l {parent=calf_l, loc=<-35.813 -0.0967 0.0397>, rot=<0.0237 0.0095 -0.0035 -0.9997>}
 thigh_twist_01_l {parent=thigh_l, loc=<-21.6303 0 0>, rot=<0.0013 0 0 -1>}
 thigh_r {parent=pelvis, loc=<-2.5693 -0.0109 7.7501>, rot=<0.0572 -0.0012 0.9982 0.0155>}
 calf_r {parent=thigh_r, loc=<39.0484 0 0>, rot=<0.0006 -0.0161 -0.0391 -0.9991>}
 foot_r {parent=calf_r, loc=<42.561 -0.1077 -0.4292>, rot=<0.0418 -0.0455 0.0343 -0.9975>}
 ball_r {parent=foot_r, loc=<6.1599 -13.8264 -0.6806>, rot=<0.0001 -0.0001 -0.7063 -0.7079>}
 calf_twist_02_r {parent=calf_r, loc=<18.8064 0.0967 -0.0397>, rot=<0.0237 0.0095 -0.0035 -0.9997>}
 calf_twist_01_r {parent=calf_r, loc=<35.813 0.0967 -0.0397>, rot=<0.0237 0.0095 -0.0035 -0.9997>}
 thigh_twist_01_r {parent=thigh_r, loc=<21.6303 0 0>, rot=<0.0013 0 0 -1>}
]
variables [
 IsCrouching {type=bool, default=False}
 IsCrouchingAlpha {type=float, default=0}
 IsMoving2d {type=bool, default=False}
 IsMoving2dAlpha {type=float, default=0}
 IsNotCrouchingAlpha {type=float, default=0}
 LookUpAndCrouchAlpha {type=double, default=0}
 isTargeting {type=bool, default=False}
 isTargetingAlpha {type=float, default=0}
]
