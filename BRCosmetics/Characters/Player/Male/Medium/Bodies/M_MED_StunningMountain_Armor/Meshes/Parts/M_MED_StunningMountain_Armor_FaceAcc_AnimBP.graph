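# Post-physics hair AnimBP for the StunningMountain Armor FaceAcc mesh, animating back, side, bang and strand hair bones on the player character.
# A Control Rig processes the pose copied from the parent mesh, feeding two near-duplicate AnimDynamics chains.
# bIsInFrontEnd selects world-space simulation, otherwise root-relative.
# Master hair bones then receive corrective transforms weighted by jump/fall, spine-to-head look-up and crouched look-up alphas computed each update.
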AnimBlueprint /BRCosmetics/Characters/Player/Male/Medium/Bodies/M_MED_StunningMountain_Armor/Meshes/Parts/M_MED_StunningMountain_Armor_FaceAcc_AnimBP
target_skeleton=&/M_MED_StunningMountain_Armor_FaceAcc_Skeleton
graphs [
 AnimGraph {nodes [
   n1 = Anim Dynamics - Bone: dyn_hairBack_00(ComponentPose=Local To Component(LocalPose=Use cached pose 'ControlRigPose'()), Alpha=1, ExternalForce="0,0,0") {
    angularDampingOverride=1
    angularSpringConstant=1500
    bAngularSpring=true
    bOverrideAngularDamping=true
    bUsePlanarLimit=false
    gravityScale=0.35
    simulationSpace=World
    alphaBoolBlend {blendOption=Linear}
    boundBone {boneName=dyn_hairBack_00}
    componentAppliedLinearAccClamp=<150 150 200>
    componentLinearAccScale=<0.1 0.1 0.1>
    componentLinearVelScale=<0.3 0.3 0.3>
    physicsBodyDefinitions [{
      collisionType=CustomSphere
      sphereCollisionRadius=2
      boundBone {boneName=dyn_hairBack_00}
      boxExtents=<10 10 10>
      constraintSetup {angularLimitsMax=<10 0 10>, angularLimitsMin=<-10 -15 -10>, angularTarget=<0 0 0>}
      localJointOffset=<1 0 0>
     }]
    planarLimits [{drivingBone {boneName=head}, planeTransform {location=<0 -5 0>, rotation=<rot 0 15 -89.9999>}}]
   }
   n2 = "Constraint - dyn_hairBack_01 (RRR)"!(ComponentPose=@n1, Alpha=1, ConstraintWeights_0=0.4, ConstraintWeights_1:float=0.4, ConstraintWeights_2:float=0.2) -> (Pose) {
    alphaBoolBlend {blendOption=Linear}
    boneToModify {boneName=dyn_hairBack_01}
    constraintSetup [
     {transformType=Rotation, targetBone {boneName=dyn_hair_00_l}}
     {transformType=Rotation, targetBone {boneName=dyn_hair_00_r}}
     {transformType=Rotation, targetBone {boneName=spine_05}}
    ]
    constraintWeights [1,1,1]
   }
   n3 = Anim Dynamics - Bone: dyn_hairBack_01(ComponentPose=@n2, Alpha=1, ExternalForce="0,0,0") {
    angularDampingOverride=1
    angularSpringConstant=150
    bAngularSpring=true
    bOverrideAngularDamping=true
    bUsePlanarLimit=false
    gravityScale=0.35
    simulationSpace=World
    alphaBoolBlend {blendOption=Linear}
    boundBone {boneName=dyn_hairBack_01}
    componentAppliedLinearAccClamp=<150 150 200>
    componentLinearAccScale=<0.1 0.1 0.1>
    componentLinearVelScale=<0.3 0.3 0.3>
    physicsBodyDefinitions [{
      collisionType=CustomSphere
      sphereCollisionRadius=2
      boundBone {boneName=dyn_hairBack_01}
      boxExtents=<12 12 11>
      constraintSetup {angularLimitsMax=<0 0 0>, angularLimitsMin=<-10 -25 -5>, angularTarget=<0 0 0>}
      localJointOffset=<1 0 0>
     }]
    planarLimits [
     {drivingBone {boneName=head}, planeTransform {location=<0 -6 0>, rotation=<rot 0 25 -89.9999>}}
     {drivingBone {boneName=spine_05}, planeTransform {location=<0 0 20>, rotation=<rot 0 35 -50>}}
    ]
    sphericalLimits [{limitRadius=6.5, drivingBone {boneName=neck_02}, sphereLocalOffset=<0 0 0>}]
   }
   n4 = Anim Dynamics - Bone: dyn_hair_00_l(ComponentPose=@n3, Alpha=1, ExternalForce="0,0,0") {
    angularDampingOverride=1
    angularSpringConstant=150
    bAngularSpring=true
    bOverrideAngularDamping=true
    bUsePlanarLimit=false
    gravityScale=0.35
    simulationSpace=World
    alphaBoolBlend {blendOption=Linear}
    boundBone {boneName=dyn_hair_00_l}
    componentAppliedLinearAccClamp=<150 150 200>
    componentLinearAccScale=<0.1 0.1 0.1>
    componentLinearVelScale=<0.3 0.3 0.3>
    physicsBodyDefinitions [{
      collisionType=CustomSphere
      sphereCollisionRadius=2
      boundBone {boneName=dyn_hair_00_l}
      boxExtents=<15 15 15>
      constraintSetup {angularLimitsMax=<10 -1 10>, angularLimitsMin=<-10 -20 -10>, angularTarget=<0 0 0>}
      localJointOffset=<1 0 0>
     }]
    planarLimits [{drivingBone {boneName=head}, planeTransform {location=<-0.0545 -5.8503 -3.3295>, rotation=<rot 4.1976 1.7729 -169.2649>}}]
   }
   n5 = Anim Dynamics - Bone: dyn_hair_01_00_l(ComponentPose=@n4, Alpha=1, ExternalForce="0,0,0") {
    angularDampingOverride=1
    angularSpringConstant=150
    bAngularSpring=true
    bOverrideAngularDamping=true
    bUsePlanarLimit=false
    gravityScale=0.35
    simulationSpace=World
    alphaBoolBlend {blendOption=Linear}
    boundBone {boneName=dyn_hair_01_00_l}
    componentAppliedLinearAccClamp=<150 150 200>
    componentLinearAccScale=<0.1 0.1 0.1>
    componentLinearVelScale=<0.3 0.3 0.3>
    physicsBodyDefinitions [{
      collisionType=CustomSphere
      sphereCollisionRadius=2
      boundBone {boneName=dyn_hair_01_00_l}
      boxExtents=<15 15 15>
      constraintSetup {angularLimitsMax=<10 10 5>, angularLimitsMin=<-10 -20 -5>, angularTarget=<0 0 0>}
      localJointOffset=<1 0 0>
     }]
    planarLimits [{drivingBone {boneName=neck_01}, planeTransform {location=<0.6278 -3.3144 -5.1813>, rotation=<rot 4.1426 51.0187 -173.3935>}}]
    sphericalLimits [{limitRadius=6.5, drivingBone {boneName=neck_02}, sphereLocalOffset=<0 0 0>}]
   }
   n6 = Anim Dynamics - Bone: dyn_hair_00_r(ComponentPose=@n5, Alpha=1, ExternalForce="0,0,0") {
    angularDampingOverride=1
    angularSpringConstant=150
    bAngularSpring=true
    bOverrideAngularDamping=true
    bUsePlanarLimit=false
    gravityScale=0.35
    simulationSpace=World
    alphaBoolBlend {blendOption=Linear}
    boundBone {boneName=dyn_hair_00_r}
    componentAppliedLinearAccClamp=<150 150 200>
    componentLinearAccScale=<0.1 0.1 0.1>
    componentLinearVelScale=<0.3 0.3 0.3>
    physicsBodyDefinitions [{
      collisionType=CustomSphere
      sphereCollisionRadius=2
      boundBone {boneName=dyn_hair_00_r}
      boxExtents=<15 15 15>
      constraintSetup {angularLimitsMax=<10 -1 10>, angularLimitsMin=<-10 -20 -10>, angularTarget=<0 0 0>}
      localJointOffset=<1 0 0>
     }]
    planarLimits [{drivingBone {boneName=head}, planeTransform {location=<-0.0545 -5.8503 -3.3295>, rotation=<rot 4.1976 1.7729 -169.2649>}}]
   }
   n7 = Anim Dynamics - Bone: dyn_hair_01_00_r(ComponentPose=@n6, Alpha=1, ExternalForce="0,0,0") {
    angularDampingOverride=1
    angularSpringConstant=150
    bAngularSpring=true
    bOverrideAngularDamping=true
    bUsePlanarLimit=false
    bUseSphericalLimits=true
    gravityScale=0.35
    simulationSpace=World
    alphaBoolBlend {blendOption=Linear}
    boundBone {boneName=dyn_hair_01_00_r}
    componentAppliedLinearAccClamp=<150 150 200>
    componentLinearAccScale=<0.1 0.1 0.1>
    componentLinearVelScale=<0.3 0.3 0.3>
    physicsBodyDefinitions [{
      collisionType=CustomSphere
      sphereCollisionRadius=2
      boundBone {boneName=dyn_hair_01_00_r}
      boxExtents=<15 15 15>
      constraintSetup {angularLimitsMax=<10 10 5>, angularLimitsMin=<-10 -20 -5>, angularTarget=<0 0 0>}
      localJointOffset=<1 0 0>
     }]
    planarLimits [{drivingBone {boneName=neck_01}, planeTransform {location=<0.6278 -3.3144 -5.1813>, rotation=<rot 4.1426 51.0187 -173.3935>}}]
    sphericalLimits [
     {limitRadius=6.5, drivingBone {boneName=neck_02}, sphereLocalOffset=<0 0 0>}
     {limitRadius=8, drivingBone {boneName=upperarm_r}, sphereLocalOffset=<0 0 0>}
    ]
   }
   n8 = Anim Dynamics - Bone: dyn_hairBang_01_l(ComponentPose=@n7, Alpha=0.5, ExternalForce="0,0,0") {
    angularDampingOverride=0.9999
    angularSpringConstant=3000
    bAngularSpring=true
    bGravityOverrideInSimSpace=true
    bOverrideAngularDamping=true
    bUseSphericalLimits=true
    gravityScale=0.35
    simulationSpace=World
    alphaBoolBlend {blendOption=Linear}
    boundBone {boneName=dyn_hairBang_01_l}
    componentAppliedLinearAccClamp=<300 300 300>
    componentLinearAccScale=<0.15 0.15 0.15>
    componentLinearVelScale=<0.3 0.3 0.3>
    physicsBodyDefinitions [{
      sphereCollisionRadius=0
      boundBone {boneName=dyn_hairBang_01_l}
      boxExtents=<3 3 3>
      constraintSetup {angularLimitsMax=<0 0 35>, angularLimitsMin=<0 -20 -2>, angularTarget=<0 0 0>}
      localJointOffset=<0.2 0 0>
     }]
    relativeSpaceBone {boneName=head}
   }
   n9 = Anim Dynamics - Bone: dyn_hairBang_02_00_l(ComponentPose=@n8, Alpha=0.8, ExternalForce="0,0,0") {
    angularDampingOverride=0.9999
    angularSpringConstant=3000
    bAngularSpring=true
    bGravityOverrideInSimSpace=true
    bOverrideAngularDamping=true
    gravityScale=0.35
    simulationSpace=World
    alphaBoolBlend {blendOption=Linear}
    boundBone {boneName=dyn_hairBang_02_00_l}
    componentAppliedLinearAccClamp=<3000 3000 3000>
    componentLinearAccScale=<0.3 0.3 0.3>
    componentLinearVelScale=<0.3 0.3 0.3>
    physicsBodyDefinitions [{
      sphereCollisionRadius=0
      boundBone {boneName=dyn_hairBang_02_00_l}
      boxExtents=<4 4 4>
      constraintSetup {angularLimitsMax=<5 0 30>, angularLimitsMin=<0 -20 -10>, angularTarget=<0 0 0>}
      localJointOffset=<1 0 0>
     }]
   }
   n10 = Anim Dynamics - Bone: dyn_hairBang_03_00_l(ComponentPose=@n9, Alpha=0.8, ExternalForce="0,0,0") {
    angularDampingOverride=0.9999
    angularSpringConstant=3000
    bAngularSpring=true
    bGravityOverrideInSimSpace=true
    bOverrideAngularDamping=true
    gravityScale=0.35
    simulationSpace=World
    alphaBoolBlend {blendOption=Linear}
    boundBone {boneName=dyn_hairBang_03_00_l}
    componentAppliedLinearAccClamp=<3000 3000 3000>
    componentLinearAccScale=<0.3 0.3 0.3>
    componentLinearVelScale=<0.3 0.3 0.3>
    physicsBodyDefinitions [{
      sphereCollisionRadius=0
      boundBone {boneName=dyn_hairBang_03_00_l}
      boxExtents=<4 4 4>
      constraintSetup {angularLimitsMax=<5 0 30>, angularLimitsMin=<0 -20 -10>, angularTarget=<0 0 0>}
      localJointOffset=<0.5 0 0>
     }]
   }
   n11 = Anim Dynamics - Bone: dyn_hairBang2_02_l(ComponentPose=@n10, Alpha=1, ExternalForce="0,0,0") {
    angularDampingOverride=0.9999
    angularSpringConstant=3000
    bAngularSpring=true
    bGravityOverrideInSimSpace=true
    bOverrideAngularDamping=true
    bUseSphericalLimits=true
    gravityScale=0.35
    simulationSpace=World
    alphaBoolBlend {blendOption=Linear}
    boundBone {boneName=dyn_hairBang2_02_l}
    componentAppliedLinearAccClamp=<300 300 300>
    componentLinearAccScale=<0.15 0.15 0.15>
    componentLinearVelScale=<0.3 0.3 0.3>
    physicsBodyDefinitions [{
      sphereCollisionRadius=0
      boundBone {boneName=dyn_hairBang2_02_l}
      boxExtents=<3 3 3>
      constraintSetup {angularLimitsMax=<5 0 30>, angularLimitsMin=<0 -20 -5>, angularTarget=<0 0 0>}
      localJointOffset=<0.2 0 0>
     }]
    relativeSpaceBone {boneName=head}
   }
   n12 = Anim Dynamics - Bone: dyn_hairBang2_03_l(ComponentPose=@n11, Alpha=1, ExternalForce="0,0,0") {
    angularDampingOverride=0.9999
    angularSpringConstant=3000
    bAngularSpring=true
    bGravityOverrideInSimSpace=true
    bOverrideAngularDamping=true
    gravityScale=0.35
    simulationSpace=World
    alphaBoolBlend {blendOption=Linear}
    boundBone {boneName=dyn_hairBang2_03_l}
    componentAppliedLinearAccClamp=<3000 3000 3000>
    componentLinearAccScale=<0.3 0.3 0.3>
    componentLinearVelScale=<0.3 0.3 0.3>
    physicsBodyDefinitions [{
      sphereCollisionRadius=0
      boundBone {boneName=dyn_hairBang2_03_l}
      boxExtents=<4 4 4>
      constraintSetup {angularLimitsMax=<5 0 30>, angularLimitsMin=<0 -20 -5>, angularTarget=<0 0 0>}
      localJointOffset=<0.5 0 0>
     }]
   }
   n13 = Anim Dynamics - Bone: dyn_hairBang_01_r(ComponentPose=@n12, Alpha=0.5, ExternalForce="0,0,0") {
    angularDampingOverride=0.9999
    angularSpringConstant=3000
    bAngularSpring=true
    bGravityOverrideInSimSpace=true
    bOverrideAngularDamping=true
    bUsePlanarLimit=false
    bUseSphericalLimits=true
    gravityScale=0.35
    simulationSpace=World
    alphaBoolBlend {blendOption=Linear}
    boundBone {boneName=dyn_hairBang_01_r}
    componentAppliedLinearAccClamp=<300 300 300>
    componentLinearAccScale=<0.15 0.15 0.15>
    componentLinearVelScale=<0.3 0.3 0.3>
    physicsBodyDefinitions [{
      sphereCollisionRadius=0
      boundBone {boneName=dyn_hairBang_01_r}
      boxExtents=<5 5 5>
      constraintSetup {angularLimitsMax=<0 0 15>, angularLimitsMin=<0 -20 -15>, angularTarget=<0 0 0>}
      localJointOffset=<0.2 0 0>
     }]
    planarLimits [{drivingBone {boneName=head}, planeTransform {location=<0 11 0>, rotation=<rot 0 0 90>}}]
    relativeSpaceBone {boneName=head}
    sphericalLimits [{limitRadius=9.9, drivingBone {boneName=head}, sphereLocalOffset=<9 2 0>}]
   }
   n14 = Anim Dynamics - Bone: dyn_hairBang_02_00_r(ComponentPose=@n13, Alpha=0.5, ExternalForce="0,0,0") {
    angularDampingOverride=0.9999
    angularSpringConstant=3000
    bAngularSpring=true
    bGravityOverrideInSimSpace=true
    bOverrideAngularDamping=true
    bUsePlanarLimit=false
    bUseSphericalLimits=true
    gravityScale=0.35
    simulationSpace=World
    alphaBoolBlend {blendOption=Linear}
    boundBone {boneName=dyn_hairBang_02_00_r}
    componentAppliedLinearAccClamp=<3000 3000 3000>
    componentLinearAccScale=<0.3 0.3 0.3>
    componentLinearVelScale=<0.3 0.3 0.3>
    physicsBodyDefinitions [{
      sphereCollisionRadius=0
      boundBone {boneName=dyn_hairBang_02_00_r}
      boxExtents=<4 4 4>
      constraintSetup {angularLimitsMax=<5 0 30>, angularLimitsMin=<0 -20 -30>, angularTarget=<0 0 0>}
      localJointOffset=<0.5 0 0>
     }]
    planarLimits [{drivingBone {boneName=head}, planeTransform {location=<0 11 0>, rotation=<rot 0 0 90>}}]
    sphericalLimits [{limitRadius=10, drivingBone {boneName=head}, sphereLocalOffset=<9 2 0>}]
   }
   n15 = Anim Dynamics - Bone: dyn_hairStrand_00_l(ComponentPose=@n14, Alpha=1, ExternalForce="0,0,0") {
    angularDampingOverride=0.9999
    angularSpringConstant=1500
    bAngularSpring=true
    bGravityOverrideInSimSpace=true
    bOverrideAngularDamping=true
    bUseSphericalLimits=true
    gravityScale=0.35
    simulationSpace=World
    alphaBoolBlend {blendOption=Linear}
    boundBone {boneName=dyn_hairStrand_00_l}
    componentAppliedLinearAccClamp=<300 300 300>
    componentLinearAccScale=<0.15 0.15 0.15>
    componentLinearVelScale=<0.3 0.3 0.3>
    physicsBodyDefinitions [{
      sphereCollisionRadius=0
      boundBone {boneName=dyn_hairStrand_00_l}
      boxExtents=<8 8 8>
      constraintSetup {angularLimitsMax=<0 0 15>, angularLimitsMin=<0 -20 -2>, angularTarget=<0 0 0>}
      localJointOffset=<0.2 0 0>
     }]
    relativeSpaceBone {boneName=head}
   }
   n16 = Anim Dynamics - Bone: dyn_hairStrand_01_l(ComponentPose=@n15, Alpha=1, ExternalForce="0,0,0") {
    angularDampingOverride=0.9999
    angularSpringConstant=1500
    bAngularSpring=true
    bGravityOverrideInSimSpace=true
    bOverrideAngularDamping=true
    gravityScale=0.35
    simulationSpace=World
    alphaBoolBlend {blendOption=Linear}
    boundBone {boneName=dyn_hairStrand_01_l}
    componentAppliedLinearAccClamp=<3000 3000 3000>
    componentLinearAccScale=<0.3 0.3 0.3>
    componentLinearVelScale=<0.3 0.3 0.3>
    physicsBodyDefinitions [{
      sphereCollisionRadius=0
      boundBone {boneName=dyn_hairStrand_01_l}
      boxExtents=<8 8 8>
      constraintSetup {angularLimitsMax=<5 0 30>, angularLimitsMin=<0 -20 -5>, angularTarget=<0 0 0>}
      localJointOffset=<1 0 0>
     }]
   }
   n17 = Anim Dynamics - Bone: dyn_hairStrand_00_r(ComponentPose=@n16, Alpha=1, ExternalForce="0,0,0") {
    angularDampingOverride=0.9999
    angularSpringConstant=1500
    bAngularSpring=true
    bGravityOverrideInSimSpace=true
    bOverrideAngularDamping=true
    bUseSphericalLimits=true
    gravityScale=0.35
    simulationSpace=World
    alphaBoolBlend {blendOption=Linear}
    boundBone {boneName=dyn_hairStrand_00_r}
    componentAppliedLinearAccClamp=<300 300 300>
    componentLinearAccScale=<0.15 0.15 0.15>
    componentLinearVelScale=<0.3 0.3 0.3>
    physicsBodyDefinitions [{
      sphereCollisionRadius=0
      boundBone {boneName=dyn_hairStrand_00_r}
      boxExtents=<8 8 8>
      constraintSetup {angularLimitsMax=<0 0 15>, angularLimitsMin=<0 -20 -2>, angularTarget=<0 0 0>}
      localJointOffset=<0.2 0 0>
     }]
    relativeSpaceBone {boneName=head}
   }
   n18 = Anim Dynamics - Bone: dyn_hairStrand_01_r(ComponentPose=@n17, Alpha=1, ExternalForce="0,0,0") {
    angularDampingOverride=0.9999
    angularSpringConstant=1500
    bAngularSpring=true
    bGravityOverrideInSimSpace=true
    bOverrideAngularDamping=true
    gravityScale=0.35
    simulationSpace=World
    alphaBoolBlend {blendOption=Linear}
    boundBone {boneName=dyn_hairStrand_01_r}
    componentAppliedLinearAccClamp=<3000 3000 3000>
    componentLinearAccScale=<0.3 0.3 0.3>
    componentLinearVelScale=<0.3 0.3 0.3>
    physicsBodyDefinitions [{
      sphereCollisionRadius=0
      boundBone {boneName=dyn_hairStrand_01_r}
      boxExtents=<8 8 8>
      constraintSetup {angularLimitsMax=<5 0 30>, angularLimitsMin=<0 -20 -5>, angularTarget=<0 0 0>}
      localJointOffset=<1 0 0>
     }]
   }
   n19 = Component To Local(ComponentPose=@n18)
   n20 = Anim Dynamics - Bone: dyn_hairBack_00(ComponentPose=Local To Component(LocalPose=Use cached pose 'ControlRigPose'()), Alpha=1, ExternalForce="0,0,0") {
    angularDampingOverride=1
    angularSpringConstant=1500
    bAngularSpring=true
    bOverrideAngularDamping=true
    bUsePlanarLimit=false
    gravityScale=0.35
    simulationSpace=RootRelative
    alphaBoolBlend {blendOption=Linear}
    boundBone {boneName=dyn_hairBack_00}
    componentAppliedLinearAccClamp=<150 150 200>
    componentLinearAccScale=<0.1 0.1 0.1>
    componentLinearVelScale=<0.3 0.3 0.3>
    physicsBodyDefinitions [{
      collisionType=CustomSphere
      sphereCollisionRadius=2
      boundBone {boneName=dyn_hairBack_00}
      boxExtents=<10 10 10>
      constraintSetup {angularLimitsMax=<10 0 10>, angularLimitsMin=<-10 -15 -10>, angularTarget=<0 0 0>}
      localJointOffset=<1 0 0>
     }]
    planarLimits [{drivingBone {boneName=head}, planeTransform {location=<0 -5 0>, rotation=<rot 0 15 -90>}}]
   }
   n21 = "Constraint - dyn_hairBack_01 (RRR)"(ComponentPose=@n20, Alpha=1, ConstraintWeights_0=0.4, ConstraintWeights_1=0.4, ConstraintWeights_2=0.2) {
    alphaBoolBlend {blendOption=Linear}
    boneToModify {boneName=dyn_hairBack_01}
    constraintSetup [
     {transformType=Rotation, targetBone {boneName=dyn_hair_00_l}}
     {transformType=Rotation, targetBone {boneName=dyn_hair_00_r}}
     {transformType=Rotation, targetBone {boneName=spine_05}}
    ]
    constraintWeights [1,1,1]
   }
   n22 = Anim Dynamics - Bone: dyn_hairBack_01(ComponentPose=@n21, Alpha=1, ExternalForce="0,0,0") {
    angularDampingOverride=1
    angularSpringConstant=150
    bAngularSpring=true
    bOverrideAngularDamping=true
    bUsePlanarLimit=false
    gravityScale=0.35
    simulationSpace=RootRelative
    alphaBoolBlend {blendOption=Linear}
    boundBone {boneName=dyn_hairBack_01}
    componentAppliedLinearAccClamp=<150 150 200>
    componentLinearAccScale=<0.1 0.1 0.1>
    componentLinearVelScale=<0.3 0.3 0.3>
    physicsBodyDefinitions [{
      collisionType=CustomSphere
      sphereCollisionRadius=2
      boundBone {boneName=dyn_hairBack_01}
      boxExtents=<12 12 11>
      constraintSetup {angularLimitsMax=<0 0 0>, angularLimitsMin=<-10 -25 -5>, angularTarget=<0 0 0>}
      localJointOffset=<1 0 0>
     }]
    planarLimits [
     {drivingBone {boneName=head}, planeTransform {location=<0 -6 0>, rotation=<rot 0 25 -89.9999>}}
     {drivingBone {boneName=spine_05}, planeTransform {location=<0 0 20>, rotation=<rot 0 35 -50>}}
    ]
    sphericalLimits [{limitRadius=6.5, drivingBone {boneName=neck_02}, sphereLocalOffset=<0 0 0>}]
   }
   n23 = Anim Dynamics - Bone: dyn_hair_00_l(ComponentPose=@n22, Alpha=1, ExternalForce="0,0,0") {
    angularDampingOverride=1
    angularSpringConstant=150
    bAngularSpring=true
    bOverrideAngularDamping=true
    bUsePlanarLimit=false
    gravityScale=0.35
    simulationSpace=RootRelative
    alphaBoolBlend {blendOption=Linear}
    boundBone {boneName=dyn_hair_00_l}
    componentAppliedLinearAccClamp=<150 150 200>
    componentLinearAccScale=<0.1 0.1 0.1>
    componentLinearVelScale=<0.3 0.3 0.3>
    physicsBodyDefinitions [{
      collisionType=CustomSphere
      sphereCollisionRadius=2
      boundBone {boneName=dyn_hair_00_l}
      boxExtents=<10 10 10>
      constraintSetup {angularLimitsMax=<10 -1 10>, angularLimitsMin=<-10 -20 -10>, angularTarget=<0 0 0>}
      localJointOffset=<1 0 0>
     }]
    planarLimits [{drivingBone {boneName=head}, planeTransform {location=<-0.0545 -5.8503 -3.3295>, rotation=<rot 4.1976 1.7729 -169.2649>}}]
   }
   n24 = Anim Dynamics - Bone: dyn_hair_01_00_l(ComponentPose=@n23, Alpha=1, ExternalForce="0,0,0") {
    angularDampingOverride=1
    angularSpringConstant=150
    bAngularSpring=true
    bOverrideAngularDamping=true
    bUsePlanarLimit=false
    gravityScale=0.35
    simulationSpace=RootRelative
    alphaBoolBlend {blendOption=Linear}
    boundBone {boneName=dyn_hair_01_00_l}
    componentAppliedLinearAccClamp=<150 150 200>
    componentLinearAccScale=<0.1 0.1 0.1>
    componentLinearVelScale=<0.3 0.3 0.3>
    physicsBodyDefinitions [{
      collisionType=CustomSphere
      sphereCollisionRadius=2
      boundBone {boneName=dyn_hair_01_00_l}
      boxExtents=<12 12 11>
      constraintSetup {angularLimitsMax=<10 10 5>, angularLimitsMin=<-10 -20 -5>, angularTarget=<0 0 0>}
      localJointOffset=<1 0 0>
     }]
    planarLimits [{drivingBone {boneName=neck_01}, planeTransform {location=<0.6278 -3.3144 -5.1813>, rotation=<rot 4.1426 51.0187 -173.3935>}}]
    sphericalLimits [{limitRadius=6.5, drivingBone {boneName=neck_02}, sphereLocalOffset=<0 0 0>}]
   }
   n25 = Anim Dynamics - Bone: dyn_hair_00_r(ComponentPose=@n24, Alpha=1, ExternalForce="0,0,0") {
    angularDampingOverride=1
    angularSpringConstant=150
    bAngularSpring=true
    bOverrideAngularDamping=true
    bUsePlanarLimit=false
    gravityScale=0.35
    simulationSpace=RootRelative
    alphaBoolBlend {blendOption=Linear}
    boundBone {boneName=dyn_hair_00_r}
    componentAppliedLinearAccClamp=<150 150 200>
    componentLinearAccScale=<0.1 0.1 0.1>
    componentLinearVelScale=<0.3 0.3 0.3>
    physicsBodyDefinitions [{
      collisionType=CustomSphere
      sphereCollisionRadius=2
      boundBone {boneName=dyn_hair_00_r}
      boxExtents=<10 10 10>
      constraintSetup {angularLimitsMax=<10 -1 10>, angularLimitsMin=<-10 -20 -10>, angularTarget=<0 0 0>}
      localJointOffset=<1 0 0>
     }]
    planarLimits [{drivingBone {boneName=head}, planeTransform {location=<-0.0545 -5.8503 -3.3295>, rotation=<rot 4.1976 1.7729 -169.2649>}}]
   }
   n26 = Anim Dynamics - Bone: dyn_hair_01_00_r(ComponentPose=@n25, Alpha=1, ExternalForce="0,0,0") {
    angularDampingOverride=1
    angularSpringConstant=150
    bAngularSpring=true
    bOverrideAngularDamping=true
    bUsePlanarLimit=false
    bUseSphericalLimits=true
    gravityScale=0.35
    simulationSpace=RootRelative
    alphaBoolBlend {blendOption=Linear}
    boundBone {boneName=dyn_hair_01_00_r}
    componentAppliedLinearAccClamp=<150 150 200>
    componentLinearAccScale=<0.1 0.1 0.1>
    componentLinearVelScale=<0.3 0.3 0.3>
    physicsBodyDefinitions [{
      collisionType=CustomSphere
      sphereCollisionRadius=2
      boundBone {boneName=dyn_hair_01_00_r}
      boxExtents=<12 12 11>
      constraintSetup {angularLimitsMax=<10 10 5>, angularLimitsMin=<-10 -20 -5>, angularTarget=<0 0 0>}
      localJointOffset=<1 0 0>
     }]
    planarLimits [{drivingBone {boneName=neck_01}, planeTransform {location=<0.6278 -3.3144 -5.1813>, rotation=<rot 4.1426 51.0187 -173.3935>}}]
    sphericalLimits [
     {limitRadius=6.5, drivingBone {boneName=neck_02}, sphereLocalOffset=<0 0 0>}
     {limitRadius=8, drivingBone {boneName=upperarm_r}, sphereLocalOffset=<0 0 0>}
    ]
   }
   n27 = Anim Dynamics - Bone: dyn_hairBang_01_l(ComponentPose=@n26, Alpha=0.5, ExternalForce="0,0,0") {
    angularDampingOverride=0.9999
    angularSpringConstant=3000
    bAngularSpring=true
    bGravityOverrideInSimSpace=true
    bOverrideAngularDamping=true
    bUseSphericalLimits=true
    gravityScale=0.35
    simulationSpace=RootRelative
    alphaBoolBlend {blendOption=Linear}
    boundBone {boneName=dyn_hairBang_01_l}
    componentAppliedLinearAccClamp=<300 300 300>
    componentLinearAccScale=<0.15 0.15 0.15>
    componentLinearVelScale=<0.3 0.3 0.3>
    physicsBodyDefinitions [{
      sphereCollisionRadius=0
      boundBone {boneName=dyn_hairBang_01_l}
      boxExtents=<3 3 3>
      constraintSetup {angularLimitsMax=<0 0 35>, angularLimitsMin=<0 -20 -2>, angularTarget=<0 0 0>}
      localJointOffset=<0.2 0 0>
     }]
    relativeSpaceBone {boneName=head}
   }
   n28 = Anim Dynamics - Bone: dyn_hairBang_02_00_l(ComponentPose=@n27, Alpha=0.8, ExternalForce="0,0,0") {
    angularDampingOverride=0.9999
    angularSpringConstant=3000
    bAngularSpring=true
    bGravityOverrideInSimSpace=true
    bOverrideAngularDamping=true
    gravityScale=0.35
    simulationSpace=RootRelative
    alphaBoolBlend {blendOption=Linear}
    boundBone {boneName=dyn_hairBang_02_00_l}
    componentAppliedLinearAccClamp=<3000 3000 3000>
    componentLinearAccScale=<0.3 0.3 0.3>
    componentLinearVelScale=<0.3 0.3 0.3>
    physicsBodyDefinitions [{
      sphereCollisionRadius=0
      boundBone {boneName=dyn_hairBang_02_00_l}
      boxExtents=<4 4 4>
      constraintSetup {angularLimitsMax=<5 0 30>, angularLimitsMin=<0 -20 -10>, angularTarget=<0 0 0>}
      localJointOffset=<1 0 0>
     }]
   }
   n29 = Anim Dynamics - Bone: dyn_hairBang_03_00_l(ComponentPose=@n28, Alpha=0.8, ExternalForce="0,0,0") {
    angularDampingOverride=0.9999
    angularSpringConstant=3000
    bAngularSpring=true
    bGravityOverrideInSimSpace=true
    bOverrideAngularDamping=true
    gravityScale=0.35
    simulationSpace=RootRelative
    alphaBoolBlend {blendOption=Linear}
    boundBone {boneName=dyn_hairBang_03_00_l}
    componentAppliedLinearAccClamp=<3000 3000 3000>
    componentLinearAccScale=<0.3 0.3 0.3>
    componentLinearVelScale=<0.3 0.3 0.3>
    physicsBodyDefinitions [{
      sphereCollisionRadius=0
      boundBone {boneName=dyn_hairBang_03_00_l}
      boxExtents=<4 4 4>
      constraintSetup {angularLimitsMax=<5 0 30>, angularLimitsMin=<0 -20 -10>, angularTarget=<0 0 0>}
      localJointOffset=<0.5 0 0>
     }]
   }
   n30 = Anim Dynamics - Bone: dyn_hairBang2_02_l(ComponentPose=@n29, Alpha=1, ExternalForce="0,0,0") {
    angularDampingOverride=0.9999
    angularSpringConstant=3000
    bAngularSpring=true
    bGravityOverrideInSimSpace=true
    bOverrideAngularDamping=true
    bUseSphericalLimits=true
    gravityScale=0.35
    simulationSpace=RootRelative
    alphaBoolBlend {blendOption=Linear}
    boundBone {boneName=dyn_hairBang2_02_l}
    componentAppliedLinearAccClamp=<300 300 300>
    componentLinearAccScale=<0.15 0.15 0.15>
    componentLinearVelScale=<0.3 0.3 0.3>
    physicsBodyDefinitions [{
      sphereCollisionRadius=0
      boundBone {boneName=dyn_hairBang2_02_l}
      boxExtents=<3 3 3>
      constraintSetup {angularLimitsMax=<5 0 30>, angularLimitsMin=<0 -20 -5>, angularTarget=<0 0 0>}
      localJointOffset=<0.2 0 0>
     }]
    relativeSpaceBone {boneName=head}
   }
   n31 = Anim Dynamics - Bone: dyn_hairBang2_03_l(ComponentPose=@n30, Alpha=1, ExternalForce="0,0,0") {
    angularDampingOverride=0.9999
    angularSpringConstant=3000
    bAngularSpring=true
    bGravityOverrideInSimSpace=true
    bOverrideAngularDamping=true
    gravityScale=0.35
    simulationSpace=RootRelative
    alphaBoolBlend {blendOption=Linear}
    boundBone {boneName=dyn_hairBang2_03_l}
    componentAppliedLinearAccClamp=<3000 3000 3000>
    componentLinearAccScale=<0.3 0.3 0.3>
    componentLinearVelScale=<0.3 0.3 0.3>
    physicsBodyDefinitions [{
      sphereCollisionRadius=0
      boundBone {boneName=dyn_hairBang2_03_l}
      boxExtents=<4 4 4>
      constraintSetup {angularLimitsMax=<5 0 30>, angularLimitsMin=<0 -20 -5>, angularTarget=<0 0 0>}
      localJointOffset=<0.5 0 0>
     }]
   }
   n32 = Anim Dynamics - Bone: dyn_hairBang_01_r(ComponentPose=@n31, Alpha=0.5, ExternalForce="0,0,0") {
    angularDampingOverride=0.9999
    angularSpringConstant=3000
    bAngularSpring=true
    bGravityOverrideInSimSpace=true
    bOverrideAngularDamping=true
    bUsePlanarLimit=false
    bUseSphericalLimits=true
    gravityScale=0.35
    simulationSpace=RootRelative
    alphaBoolBlend {blendOption=Linear}
    boundBone {boneName=dyn_hairBang_01_r}
    componentAppliedLinearAccClamp=<300 300 300>
    componentLinearAccScale=<0.15 0.15 0.15>
    componentLinearVelScale=<0.3 0.3 0.3>
    physicsBodyDefinitions [{
      sphereCollisionRadius=0
      boundBone {boneName=dyn_hairBang_01_r}
      boxExtents=<5 5 5>
      constraintSetup {angularLimitsMax=<0 0 15>, angularLimitsMin=<0 -20 -15>, angularTarget=<0 0 0>}
      localJointOffset=<0.2 0 0>
     }]
    planarLimits [{drivingBone {boneName=head}, planeTransform {location=<0 11 0>, rotation=<rot 0 0 90>}}]
    relativeSpaceBone {boneName=head}
    sphericalLimits [{limitRadius=9.9, drivingBone {boneName=head}, sphereLocalOffset=<9 2 0>}]
   }
   n33 = Anim Dynamics - Bone: dyn_hairBang_02_00_r(ComponentPose=@n32, Alpha=0.5, ExternalForce="0,0,0") {
    angularDampingOverride=0.9999
    angularSpringConstant=3000
    bAngularSpring=true
    bGravityOverrideInSimSpace=true
    bOverrideAngularDamping=true
    bUsePlanarLimit=false
    bUseSphericalLimits=true
    gravityScale=0.35
    simulationSpace=RootRelative
    alphaBoolBlend {blendOption=Linear}
    boundBone {boneName=dyn_hairBang_02_00_r}
    componentAppliedLinearAccClamp=<3000 3000 3000>
    componentLinearAccScale=<0.3 0.3 0.3>
    componentLinearVelScale=<0.3 0.3 0.3>
    physicsBodyDefinitions [{
      sphereCollisionRadius=0
      boundBone {boneName=dyn_hairBang_02_00_r}
      boxExtents=<4 4 4>
      constraintSetup {angularLimitsMax=<5 0 30>, angularLimitsMin=<0 -20 -30>, angularTarget=<0 0 0>}
      localJointOffset=<0.5 0 0>
     }]
    planarLimits [{drivingBone {boneName=head}, planeTransform {location=<0 11 0>, rotation=<rot 0 0 90>}}]
    sphericalLimits [{limitRadius=10, drivingBone {boneName=head}, sphereLocalOffset=<9 2 0>}]
   }
   n34 = Anim Dynamics - Bone: dyn_hairStrand_00_l(ComponentPose=@n33, Alpha=1, ExternalForce="0,0,0") {
    angularDampingOverride=0.9999
    angularSpringConstant=1500
    bAngularSpring=true
    bGravityOverrideInSimSpace=true
    bOverrideAngularDamping=true
    bUseSphericalLimits=true
    gravityScale=0.35
    simulationSpace=RootRelative
    alphaBoolBlend {blendOption=Linear}
    boundBone {boneName=dyn_hairStrand_00_l}
    componentAppliedLinearAccClamp=<300 300 300>
    componentLinearAccScale=<0.15 0.15 0.15>
    componentLinearVelScale=<0.3 0.3 0.3>
    physicsBodyDefinitions [{
      sphereCollisionRadius=0
      boundBone {boneName=dyn_hairStrand_00_l}
      boxExtents=<3 3 3>
      constraintSetup {angularLimitsMax=<0 0 15>, angularLimitsMin=<0 -20 -2>, angularTarget=<0 0 0>}
      localJointOffset=<0.2 0 0>
     }]
    relativeSpaceBone {boneName=head}
   }
   n35 = Anim Dynamics - Bone: dyn_hairStrand_01_l(ComponentPose=@n34, Alpha=1, ExternalForce="0,0,0") {
    angularDampingOverride=0.9999
    angularSpringConstant=1500
    bAngularSpring=true
    bGravityOverrideInSimSpace=true
    bOverrideAngularDamping=true
    gravityScale=0.35
    simulationSpace=RootRelative
    alphaBoolBlend {blendOption=Linear}
    boundBone {boneName=dyn_hairStrand_01_l}
    componentAppliedLinearAccClamp=<3000 3000 3000>
    componentLinearAccScale=<0.3 0.3 0.3>
    componentLinearVelScale=<0.3 0.3 0.3>
    physicsBodyDefinitions [{
      sphereCollisionRadius=0
      boundBone {boneName=dyn_hairStrand_01_l}
      boxExtents=<5 7.5 7.5>
      constraintSetup {angularLimitsMax=<5 0 30>, angularLimitsMin=<0 -20 -5>, angularTarget=<0 0 0>}
      localJointOffset=<1 0 0>
     }]
   }
   n36 = Anim Dynamics - Bone: dyn_hairStrand_00_r(ComponentPose=@n35, Alpha=1, ExternalForce="0,0,0") {
    angularDampingOverride=0.9999
    angularSpringConstant=1500
    bAngularSpring=true
    bGravityOverrideInSimSpace=true
    bOverrideAngularDamping=true
    bUseSphericalLimits=true
    gravityScale=0.35
    simulationSpace=RootRelative
    alphaBoolBlend {blendOption=Linear}
    boundBone {boneName=dyn_hairStrand_00_r}
    componentAppliedLinearAccClamp=<300 300 300>
    componentLinearAccScale=<0.15 0.15 0.15>
    componentLinearVelScale=<0.3 0.3 0.3>
    physicsBodyDefinitions [{
      sphereCollisionRadius=0
      boundBone {boneName=dyn_hairStrand_00_r}
      boxExtents=<3 3 3>
      constraintSetup {angularLimitsMax=<0 0 15>, angularLimitsMin=<0 -20 -2>, angularTarget=<0 0 0>}
      localJointOffset=<0.2 0 0>
     }]
    relativeSpaceBone {boneName=head}
   }
   n37 = Anim Dynamics - Bone: dyn_hairStrand_01_r(ComponentPose=@n36, Alpha=1, ExternalForce="0,0,0") {
    angularDampingOverride=0.9999
    angularSpringConstant=1500
    bAngularSpring=true
    bGravityOverrideInSimSpace=true
    bOverrideAngularDamping=true
    gravityScale=0.35
    simulationSpace=RootRelative
    alphaBoolBlend {blendOption=Linear}
    boundBone {boneName=dyn_hairStrand_01_r}
    componentAppliedLinearAccClamp=<3000 3000 3000>
    componentLinearAccScale=<0.3 0.3 0.3>
    componentLinearVelScale=<0.3 0.3 0.3>
    physicsBodyDefinitions [{
      sphereCollisionRadius=0
      boundBone {boneName=dyn_hairStrand_01_r}
      boxExtents=<5 7.5 7.5>
      constraintSetup {angularLimitsMax=<5 0 30>, angularLimitsMin=<0 -20 -5>, angularTarget=<0 0 0>}
      localJointOffset=<1 0 0>
     }]
   }
   n38 = Component To Local(ComponentPose=@n37)
   n39 = Get bIsInFrontEnd()
   n40 = Get AlphaJumping() -> (AlphaJumping:double)
   n41 = "Transform (Modify) Bone - Bone: mstr_dyn_hairBack_00"(
    ComponentPose=Local To Component(LocalPose=Blend Poses by bool(BlendPose_0=@n19, BlendPose_1=@n38, BlendTime_0=0.5, BlendTime_1=0.5, bActiveValue=@n39) {blendType=Linear})
    Alpha=@n40
    Translation="0.395,-1.000,0.521"
    Rotation="6.228,-0.153,13.605"
    Scale="1,1,1"
   ) {
    rotationMode=BMM_Additive
    translationMode=BMM_Additive
    alphaBoolBlend {blendOption=Linear}
    boneToModify {boneName=mstr_dyn_hairBack_00}
    rotation=<rot 6.228 -0.153 13.605>
    translation=<0.395 -1 0.521>
   }
   n42 = "Transform (Modify) Bone - Bone: mstr_dyn_hair_00_l"(ComponentPose=@n41, Alpha=@n40, Translation="0,0,0", Rotation="0,0,10", Scale="1,1,1") {
    rotationMode=BMM_Additive
    translationMode=BMM_Additive
    alphaBoolBlend {blendOption=Linear}
    boneToModify {boneName=mstr_dyn_hair_00_l}
    rotation {roll=10}
   }
   n43 = "Transform (Modify) Bone - Bone: mstr_dyn_hair_00_r"(ComponentPose=@n42, Alpha=@n40, Translation="0,0,0", Rotation="0,0,10", Scale="1,1,1") {
    rotationMode=BMM_Additive
    translationMode=BMM_Additive
    alphaBoolBlend {blendOption=Linear}
    boneToModify {boneName=mstr_dyn_hair_00_r}
    rotation {roll=10}
   }
   n44 = Get LookUpPoseAlpha() -> (LookUpPoseAlpha:float)
   n45 = "Transform (Modify) Bone - Bone: mstr_dyn_hairBack_00"(ComponentPose=@n43, Alpha=@n44, Translation="0,0,0", Rotation="0,0,0", Scale="0.8,1,1") {
    rotationMode=BMM_Additive
    scaleMode=BMM_Replace
    translationMode=BMM_Additive
    alphaBoolBlend {blendOption=Linear}
    boneToModify {boneName=mstr_dyn_hairBack_00}
    scale {x=0.8}
   }
   n46 = "Transform (Modify) Bone - Bone: mstr_dyn_hair_00_r"(ComponentPose=@n45, Alpha=@n44, Translation="-1,0,0", Rotation="0,0,0", Scale="0.9,1,1") {
    rotationMode=BMM_Additive
    scaleMode=BMM_Replace
    translationMode=BMM_Additive
    alphaBoolBlend {blendOption=Linear}
    boneToModify {boneName=mstr_dyn_hair_00_r}
    scale {x=0.9}
    translation {x=-1}
   }
   n47 = "Transform (Modify) Bone - Bone: mstr_dyn_hair_00_l"(ComponentPose=@n46, Alpha=@n44, Translation="0,0,0", Rotation="0,0,0", Scale="0.9,1,1") {
    rotationMode=BMM_Additive
    scaleMode=BMM_Replace
    translationMode=BMM_Additive
    alphaBoolBlend {blendOption=Linear}
    boneToModify {boneName=mstr_dyn_hair_00_l}
    scale {x=0.9}
   }
   n48 = Get LookUpCrouchAlpha() -> (LookUpCrouchAlpha:double)
   n49 = "Transform (Modify) Bone - Bone: mstr_dyn_hairBack_00"(ComponentPose=@n47, Alpha=@n48, Translation="0,0,0", Rotation="0,0,10", Scale="0.7,1,1") {
    rotationMode=BMM_Additive
    scaleMode=BMM_Replace
    translationMode=BMM_Additive
    alphaBoolBlend {blendOption=Linear}
    boneToModify {boneName=mstr_dyn_hairBack_00}
    rotation {roll=10}
    scale {x=0.7}
   }
   n50 = "Transform (Modify) Bone - Bone: mstr_dyn_hair_00_r"(ComponentPose=@n49, Alpha=@n48, Translation="0,0,0", Rotation="0,0,0", Scale="0.8,1,1") {
    rotationMode=BMM_Additive
    scaleMode=BMM_Replace
    translationMode=BMM_Additive
    alphaBoolBlend {blendOption=Linear}
    boneToModify {boneName=mstr_dyn_hair_00_r}
    scale {x=0.8}
   }
   Output Pose(Result=Component To Local(ComponentPose=@n50)) {layerGroup=None}
   n51 = Control Rig!(
    Source=Copy Pose From Mesh() {bUseAttachedParent=true}
    Alpha
    IsCrouching:bool=Get bIsCrouching()
    IsMoving2d:bool=Get bIsMoving2D()
    isTargeting:bool=Get bIsTargeting()
    LookUpAndCrouchAlpha:double=0
   ) -> (Pose) {
    alphaBoolBlend {blendOption=Linear}
    controlRigAssetReference {blueprintRigClass=&/M_MED_StunningMountain_Armor_FaceAcc_CtrlRig.M_MED_StunningMountain_Armor_FaceAcc_CtrlRig_C}
   }
   Save cached pose 'ControlRigPose'(Pose=@n51) {cacheName=ControlRigPose}
  ]}
 EventGraph {nodes [
   n1 = Event BlueprintUpdateAnimation()
   n2 = AND Boolean(A=OR Boolean(A=Get bIsJumping(), B=Get bIsFalling()), B=NOT Boolean(A=Get bIsSkyDiving()))
   > Set AlphaJumping(AlphaJumping:double=FInterpTo(Current=Get AlphaJumping(), Target=SelectFloat(A=1, B=0, bPickA=@n2), DeltaTime=@n1.DeltaTimeX, InterpSpeed=5)) -> (Output_Get:double)
   n3 = Get bIsCrouching()
   > Set CrouchAlpha(CrouchAlpha:double=FInterpTo(Current=Get CrouchAlpha() -> (CrouchAlpha:double), Target=SelectFloat(A=1, B=0, bPickA=@n3), DeltaTime=@n1.DeltaTimeX, InterpSpeed=10)) -> (Output_Get:double)
   n4 = GetAlphaFromReferenceTransformDelta(SkeletalMeshComponent=GetOwnerSkeletalMeshComponent(), BoneName=spine_05, BaseName=head, InputRange=(X=10,Y=45), TwistAxis="1,0,0")
   > n5 = Set LookUpPoseAlpha(LookUpPoseAlpha:float=@n4) -> (Output_Get:float)
   n6 = float * float(A=@n4, B=Get CrouchAlpha()) {is=title K2Node_PromotableOperator}
   > Set LookUpCrouchAlpha(LookUpCrouchAlpha:double=@n6) -> (Output_Get:double)
   Print Float!(Execute:Exec, Float:double=@n5, OptionalString:String=bp_up_, Duration:float=0.1, TextColor:Linear Color Structure=(R=0,G=0.6588,B=1,A=1)) -> (Then:Exec) {is=title K2Node_MacroInstance}
   > Print Float(Float=@n6, OptionalString=bp_up_Crocuh_, Duration=0.1, TextColor=(R=0,G=0.6588,B=1,A=1))
   > Print Float(Float=Get CrouchAlpha(), OptionalString=rouchAlpha_, Duration=0.1, TextColor=(R=0,G=0.6588,B=1,A=1))
   Get bIsCrouching()
  ]}
]
variables [
 AlphaJumping {type=number}
 CrouchAlpha {type=number}
 LookUpCrouchAlpha {type=number}
 LookUpPoseAlpha {type=number}
]
